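# Leaderboard row widget for LTM HUDs, showing a placement, player name and animated coin value.
# It starts collapsed; EventSetData collapses it for position -1, otherwise fills the texts, interpolates the coin count and makes it visible.
# PlayerStyle uses isPlayer to highlight the row with a white pill background, black text and transparent outlines on placement and coins.
# An unconnected Format Text node for a "{player}'s team" label is unused.
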WidgetBlueprint /Game/Athena/HUD/LTM/GenericLeaderboardEntry
graphs [EventGraph {nodes [
   n1 = EventSetData!() -> (OutputDelegate, then, Position:int, Name:String, Coins:int, isPlayer:bool=~) {is=title K2Node_CustomEvent}
   > n2 = Branch(Condition="Equal (Integer)"(A=@n1.Position, B=-1) {is=title K2Node_CallFunction})
   > SetVisibility(InVisibility=Collapsed)
   "SetText (Text)"(execute=@n2.else, self=Get Text_Placement() -> (Text_Placement:Common Text Object Reference), InText="To Text (Integer)"(Value=@n1.Position))
   n3 = "To Text (String)"(InString=@n1.Name)
   > "SetText (Text)"(self=Get Text_Playername() -> (Text_Playername:Common Text Object Reference), InText=@n3)
   > InterpolateToValue!(
    execute
    self:Common Numeric Text Block Object Reference=Get CommonNumericTextBlock_CoinValue() -> (CommonNumericTextBlock_CoinValue:Common Numeric Text Block Object Reference)
    TargetValue:float="To Float (Integer)"(InInt=@n1.Coins)
    MaximumInterpolationDuration:float=0.3
    MinimumChangeRate:float=1
    OutroOffset:float=0
   ) -> (then)
   > PlayerStyle!(execute, self, isPlayer:bool=@n1.isPlayer) -> (then)
   > SetVisibility(InVisibility=Visible)
   n4 = PlayerStyle!() -> (OutputDelegate, then, isPlayer:bool=~) {is=title K2Node_CustomEvent}
   > n5 = Set isPlayer(isPlayer:bool=@n4.isPlayer) -> (Output_Get:bool)
   n6 = Get isPlayer() -> (isPlayer:bool)
   SetBrushColor(
    execute=[@n9.then, @n5.then]
    self=Get CommonBorder_Selected() -> (CommonBorder_Selected:Common Border Object Reference)
    InBrushColor=Select(Option 0:Linear Color Structure=(R=0,G=0,B=0,A=0), Option 1:Linear Color Structure=(R=1,G=1,B=1,A=1), Index=@n6) -> (ReturnValue:Linear Color Structure)
   )
   > SetContentColorAndOpacity!(
    execute
    self:Border Object Reference=Get CommonBorder_Text() -> (CommonBorder_Text:Common Border Object Reference)
    InContentColorAndOpacity:Linear Color Structure=Select(Option 0=(R=1,G=1,B=1,A=1), Option 1=(R=0,G=0,B=0,A=1), Index=@n6)
   ) -> (then)
   n7 = Get CommonNumericTextBlock_CoinValue()
   > SetColorAndOpacity(self=@n7, InColorAndOpacity=Make Slate Color(SpecifiedColor=Select(Option 0=(R=1,G=1,B=1,A=1), Option 1=(R=0,G=0,B=0,A=1), Index=@n6)))
   n8 = Select(Option 0=(R=0.0052,G=0.0052,B=0.0056,A=0.6), Option 1=(R=0,G=0,B=0,A=0), Index=@n6)
   > Set Outline Color!(In:Exec, CommonTextWidget:Common Text Object Reference=@n7, Color:Linear Color Structure=@n8) -> (Out:Exec) {is=title K2Node_MacroInstance}
   > Set Outline Color(CommonTextWidget=Get Text_Placement(), Color=@n8)
   n9 = Event PreConstruct()
   Event Construct()
   > SetVisibility(InVisibility=Collapsed)
   Format Text!(Format="LOCTEXT(\"{player}\\'s team\")", player:Text=@n3) -> (Result)
  ]}]
variables [isPlayer {type=boolean}]
widgets {
 info {inheritedWidgetCount=0, parentClass=&UserWidget, rootWidgetClass=&Overlay, widgetCount=14}
 widgets [
  Overlay_0 {widgetClassPath=&Overlay, widget=Overlay_0}
  CommonBorder_Selected {
   widgetClassPath=&CommonBorder
   bIsVariable=true
   parent=Overlay_0
   widget=CommonBorder_Selected
   properties {
    background {
     bIsDynamicallyLoaded=false
     drawAs=Image
     imageType=NoImage
     mirroring=NoMirror
     resourceName=None
     resourceObject=null
     tiling=NoTile
     imageSize=<32 32>
     margin=<ltrb 0 0 0 0>
     outlineSettings {
      bUseBrushTransparency=false
      roundingType=HalfHeightRadius
      width=0
      color {colorUseRule=UseColor_Specified, specifiedColor=<rgba 0 0 0 0>}
      cornerRadii=<0 0 0 0>
     }
     tintColor {colorUseRule=UseColor_Specified, specifiedColor=<rgba 1 1 1 1>}
     uVRegion {bIsValid=false, max=<0 0>, min=<0 0>}
    }
    brushColor=<rgba 1 1 1 0>
    padding=<ltrb 10 2 20 2>
   }
   slot {class=OverlaySlot, properties {horizontalAlignment=HAlign_Fill}}
  }
  HorizontalBox_0 {
   widgetClassPath=&HorizontalBox
   parent=CommonBorder_Selected
   widget=HorizontalBox_0
   slot {class=BorderSlot, properties {padding=<ltrb 10 2 20 2>}}
  }
  CommonBorder_Text {
   widgetClassPath=&CommonBorder
   bIsVariable=true
   parent=HorizontalBox_0
   widget=CommonBorder_Text
   properties {padding=<ltrb 0 0 0 0>}
   slot {class=HorizontalBoxSlot, properties {verticalAlignment=VAlign_Top, size {sizeRule=Fill, value=1}}}
  }
  HorizontalBox_2 {
   widgetClassPath=&HorizontalBox
   parent=CommonBorder_Text
   widget=HorizontalBox_2
   slot {class=BorderSlot, properties {padding=<ltrb 0 0 0 0>}}
  }
  SizeBox_1 {
   widgetClassPath=&SizeBox
   parent=HorizontalBox_2
   widget=SizeBox_1
   properties {bOverride_WidthOverride=true, widthOverride=42.6275}
   slot {class=HorizontalBoxSlot}
  }
  ScaleBox_1 {
   widgetClassPath=&ScaleBox
   parent=SizeBox_1
   widget=ScaleBox_1
   properties {stretchDirection=DownOnly}
   slot {class=SizeBoxSlot, properties {verticalAlignment=VAlign_Center}}
  }
  Text_Placement {
   widgetClassPath=&CommonTextBlock
   bIsVariable=true
   parent=ScaleBox_1
   widget=Text_Placement
   properties {
    justification=Right
    text=--
    textTransformPolicy=ToUpper
    font {
     bForceMonospaced=false
     bMaterialIsStencil=false
     fontMaterial=null
     fontObject=&BurbankSmall
     letterSpacing=0
     monospacedWidth=1
     size=24
     skewAmount=0
     typefaceFontName=Bold
     outlineSettings {
      bApplyOutlineToDropShadows=false
      bMiteredCorners=false
      bSeparateFillAlpha=false
      outlineMaterial=null
      outlineSize=0
      outlineColor=<rgba 0 0 0 1>
     }
    }
    shadowOffset=<0 0>
   }
   slot {class=ScaleBoxSlot}
  }
  SizeBox_0 {
   widgetClassPath=&SizeBox
   parent=HorizontalBox_2
   widget=SizeBox_0
   properties {bOverride_WidthOverride=true, clipping=ClipToBounds, maxDesiredWidth=325, minDesiredWidth=300}
   slot {class=HorizontalBoxSlot, properties {verticalAlignment=VAlign_Center, size {sizeRule=Fill, value=1}}}
  }
  ScaleBox_0 {widgetClassPath=&ScaleBox, parent=SizeBox_0, widget=ScaleBox_0, properties {stretchDirection=DownOnly}, slot {class=SizeBoxSlot}}
  Text_Playername {
   widgetClassPath=&CommonTextBlock
   bIsVariable=true
   parent=ScaleBox_0
   widget=Text_Playername
   properties {
    text="{PlayerName}"
    textTransformPolicy=ToUpper
    font {
     bForceMonospaced=false
     bMaterialIsStencil=false
     fontMaterial=null
     fontObject=&BurbankSmall
     letterSpacing=0
     monospacedWidth=1
     size=20
     skewAmount=0
     typefaceFontName=Bold
     outlineSettings {
      bApplyOutlineToDropShadows=false
      bMiteredCorners=false
      bSeparateFillAlpha=false
      outlineMaterial=null
      outlineSize=0
      outlineColor=<rgba 0 0 0 1>
     }
    }
    shadowColorAndOpacity=<rgba 0.0052 0.0052 0.0056 0.4>
    shadowOffset=<2 2>
   }
   slot {class=ScaleBoxSlot, properties {horizontalAlignment=HAlign_Right}}
  }
  SizeBox_2 {
   widgetClassPath=&SizeBox
   parent=HorizontalBox_0
   widget=SizeBox_2
   properties {bOverride_WidthOverride=true, widthOverride=50}
   slot {
    class=HorizontalBoxSlot
    properties {horizontalAlignment=HAlign_Right, verticalAlignment=VAlign_Center, padding=<ltrb 20 0 0 0>, size {sizeRule=Automatic, value=0}}
   }
  }
  HorizontalBox_1 {widgetClassPath=&HorizontalBox, parent=SizeBox_2, widget=HorizontalBox_1, slot {class=SizeBoxSlot}}
  CommonNumericTextBlock_CoinValue {
   widgetClassPath=/Script/CommonUI.CommonNumericTextBlock
   bIsVariable=true
   parent=HorizontalBox_1
   widget=CommonNumericTextBlock_CoinValue
   properties {
    text=Text Block
    font {
     bForceMonospaced=false
     bMaterialIsStencil=false
     fontMaterial=null
     fontObject=&Burbank
     letterSpacing=0
     monospacedWidth=1
     size=30
     skewAmount=0
     typefaceFontName=Burbank-Black
     outlineSettings {
      bApplyOutlineToDropShadows=false
      bMiteredCorners=false
      bSeparateFillAlpha=false
      outlineMaterial=null
      outlineSize=0
      outlineColor=<rgba 0 0 0 0>
     }
    }
    shadowColorAndOpacity=<rgba 0.0052 0.0052 0.0056 0.4>
    shadowOffset=<2 2>
   }
   slot {class=HorizontalBoxSlot, properties {horizontalAlignment=HAlign_Left, verticalAlignment=VAlign_Top, padding=<ltrb 0 -5 0 -5>}}
  }
 ]
}
