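# GenericLeaderboard is a HUD widget for limited-time modes, showing up to four leaderboard rows under a header with a collapsed coin goal.
# UpdateDisplay sequences through the first four placement array indices, resetting rows without data using a default struct.
# Valid entries pass a dirty check, so only changed ones push placement, name, score and local-player flag to their row through EventSetData.
# An InvalidationBox root caches the layout.
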WidgetBlueprint /Game/Athena/HUD/LTM/GenericLeaderboard
graphs [
 EventGraph {nodes [
   n1 = Event UpdateDisplay!() -> (OutputDelegate, then, Data:Array of Fort In Game Leaderboard Placement Data Structures) {is=title K2Node_Event}
   > n2 = Sequence!(execute) -> (then_0, then_1, then_2:Exec, then_3:Exec)
   > n3 = Get Valid Array Entry!(execute:Exec, InputPin:Array of Fort In Game Leaderboard Placement Data Structures=@n1.Data, IndexToTest:int) -> (true:Exec, false:Exec, Output:"Fort In Game Leaderboard Placement Data Structure (by ref)") {is=title K2Node_MacroInstance}
   n4 = Get GenericLeaderboardEntry() -> (GenericLeaderboardEntry:Generic Leaderboard Entry Object Reference)
   Reset!(execute=@n3.false, self, ToReset:Generic Leaderboard Entry Object Reference=@n4) -> (then)
   CheckIsDirty!(execute=@n3.true, CheckStruct:"Fort In Game Leaderboard Placement Data Structure (by ref)"=@n3) -> (Dirty:Exec, Clean:Exec)
   > Update Entry!(
    execute:Exec
    self:Generic Leaderboard Entry Object Reference=@n4
    GenericPlacementData:"Fort In Game Leaderboard Placement Data Structure (by ref)"=@n3
   ) {is=title K2Node_MacroInstance}
   n5 = Get Valid Array Entry(execute=@n2.then_1, InputPin=@n1.Data, IndexToTest=1)
   n6 = Get GenericLeaderboardEntry_C_1() -> (GenericLeaderboardEntry_C_1:Generic Leaderboard Entry Object Reference)
   Reset(execute=@n5.false, ToReset=@n6)
   CheckIsDirty(execute=@n5.true, CheckStruct=@n5)
   > Update Entry(self=@n6, GenericPlacementData=@n5)
   n7 = Get Valid Array Entry(execute=@n2.then_2, InputPin=@n1.Data, IndexToTest=2)
   n8 = Get GenericLeaderboardEntry_C_2() -> (GenericLeaderboardEntry_C_2:Generic Leaderboard Entry Object Reference)
   Reset(execute=@n7.false, ToReset=@n8)
   CheckIsDirty(execute=@n7.true, CheckStruct=@n7)
   > Update Entry(self=@n8, GenericPlacementData=@n7)
   n9 = Get Valid Array Entry(execute=@n2.then_3, InputPin=@n1.Data, IndexToTest=3)
   n10 = Get GenericLeaderboardEntry_C_3() -> (GenericLeaderboardEntry_C_3:Generic Leaderboard Entry Object Reference)
   Reset(execute=@n9.false, ToReset=@n10)
   CheckIsDirty(execute=@n9.true, CheckStruct=@n9)
   > Update Entry(self=@n10, GenericPlacementData=@n9)
  ]}
 GetValidArrayEntry {nodes [
   n1 = Inputs!() -> (execute, InputPin:Array of Fort In Game Leaderboard Placement Data Structures, IndexToTest:int)
   > n2 = Branch(Condition=Is Valid Index!(TargetArray:Array of Fort In Game Leaderboard Placement Data Structures=@n1, IndexToTest:int=@n1.IndexToTest) -> (ReturnValue:bool))
   > Outputs!(
    true:Exec
    false:Exec=@n2.else
    Output:"Fort In Game Leaderboard Placement Data Structure (by ref)"="Get (a ref)"(Array:Array of Fort In Game Leaderboard Placement Data Structures=@n1, Dimension 1=@n1.IndexToTest) -> (Output:"Fort In Game Leaderboard Placement Data Structure (by ref)")
   )
  ]}
 ManuallyCheckIsDirty {nodes [
   n1 = Inputs!() -> (execute, Test:"Fort In Game Leaderboard Placement Data Structure (by ref)")
   > n2 = Branch(Condition=Break Fort in Game Leaderboard Placement Data!(FortInGameLeaderboardPlacementData:"Fort In Game Leaderboard Placement Data Structure (by ref)"=@n1) -> (DisplayName:String, Score:int, ScoreToWin:int, TeamNum:Byte, SortIndex:int, PctToWin:float=~, bAnyValueDirty:bool=~) {is=title K2Node_BreakStruct}.bAnyValueDirty)
   > Outputs!(Dirty:Exec, Clean:Exec=@n2.else)
  ]}
 Reset {nodes [
   n1 = Reset!() -> (then, ToReset:Generic Leaderboard Entry Object Reference) {is=title K2Node_FunctionEntry}
   > Update Entry(
    self=@n1
    GenericPlacementData=Make Fort in Game Leaderboard Placement Data!() -> (FortInGameLeaderboardPlacementData:Fort In Game Leaderboard Placement Data Structure) {is=title K2Node_MakeStruct}
   )
  ]}
 UpdateEntry {nodes [
   n1 = Inputs!() -> (execute, self:Generic Leaderboard Entry Object Reference, GenericPlacementData:"Fort In Game Leaderboard Placement Data Structure (by ref)")
   n2 = Break Fort in Game Leaderboard Placement Data!(FortInGameLeaderboardPlacementData=@n1.GenericPlacementData) -> (DisplayName, Score, ScoreToWin, TeamNum, Placement:int, SortIndex, bRepresentsLocalPlayer:bool=~, PctToWin=~, bAnyValueDirty=~)
   > EventSetData!(
    execute
    self:Generic Leaderboard Entry Object Reference=@n1
    Position:int=@n2.Placement
    Name:String=@n2
    Coins:int=@n2.Score
    isPlayer:bool=@n2.bRepresentsLocalPlayer
   ) -> (then)
   Outputs!()
  ]}
]
variables []
widgets {
 info {
  inheritedWidgetCount=1
  parentClass=/Script/FortniteUI.GenericLeaderboardDisplayWidgetBase
  rootWidgetClass=/Script/UMG.InvalidationBox
  widgetCount=17
 }
 widgets [
  InvalidationBox_0 {widgetClassPath=&InvalidationBox, bIsVariable=true, widget=InvalidationBox_0}
  Overlay_21 {widgetClassPath=&Overlay, parent=InvalidationBox_0, widget=Overlay_21, slot {class=PanelSlot}}
  VerticalBox_1 {
   widgetClassPath=&VerticalBox
   parent=Overlay_21
   widget=VerticalBox_1
   slot {class=OverlaySlot, properties {horizontalAlignment=HAlign_Fill, verticalAlignment=VAlign_Fill, padding=<ltrb 0 50 0 0>}}
  }
  VerticalBoxHeader {widgetClassPath=&VerticalBox, parent=VerticalBox_1, widget=VerticalBoxHeader, slot {class=VerticalBoxSlot}}
  CommonTextBlock_1 {
   widgetClassPath=&CommonTextBlock
   bIsVariable=true
   parent=VerticalBoxHeader
   widget=CommonTextBlock_1
   properties {
    style=/FortUILibrary/Styles/Typography/Foundation/TextStyle-Header-S.TextStyle-Header-S_C
    text=Leaderboard
    textTransformPolicy=ToUpper
    font {
     bForceMonospaced=false
     bMaterialIsStencil=false
     fontMaterial=null
     fontObject=&Burbank
     letterSpacing=0
     monospacedWidth=1
     size=24
     skewAmount=0
     typefaceFontName=Burbank-Black
     outlineSettings {
      bApplyOutlineToDropShadows=false
      bMiteredCorners=false
      bSeparateFillAlpha=false
      outlineMaterial=null
      outlineSize=0
      outlineColor=<rgba 0 0 0 1>
     }
    }
    shadowOffset=<0 0>
   }
   slot {class=VerticalBoxSlot, properties {horizontalAlignment=HAlign_Center}}
  }
  HorizontalBoxGoalScore {
   widgetClassPath=&HorizontalBox
   parent=VerticalBoxHeader
   widget=HorizontalBoxGoalScore
   properties {visibility=Collapsed}
   slot {class=VerticalBoxSlot, properties {horizontalAlignment=HAlign_Center}}
  }
  CommonTextBlock_0 {
   widgetClassPath=&CommonTextBlock
   parent=HorizontalBoxGoalScore
   widget=CommonTextBlock_0
   properties {
    text=Goal:
    textTransformPolicy=ToUpper
    font {
     bForceMonospaced=false
     bMaterialIsStencil=false
     fontMaterial=null
     fontObject=&BurbankSmall
     letterSpacing=0
     monospacedWidth=1
     size=18
     skewAmount=0
     typefaceFontName=Bold
     outlineSettings {
      bApplyOutlineToDropShadows=false
      bMiteredCorners=false
      bSeparateFillAlpha=false
      outlineMaterial=null
      outlineSize=0
      outlineColor=<rgba 0 0 0 1>
     }
    }
    shadowOffset=<0 0>
   }
   slot {class=HorizontalBoxSlot, properties {verticalAlignment=VAlign_Center}}
  }
  HorizontalBox_1 {
   widgetClassPath=&HorizontalBox
   parent=HorizontalBoxGoalScore
   widget=HorizontalBox_1
   slot {class=HorizontalBoxSlot, properties {horizontalAlignment=HAlign_Center, verticalAlignment=VAlign_Top, padding=<ltrb 10 0 0 0>}}
  }
  Image_48 {
   widgetClassPath=&Image
   bIsVariable=true
   parent=HorizontalBox_1
   widget=Image_48
   properties {brush {
     bIsDynamicallyLoaded=false
     drawAs=Image
     imageType=NoImage
     mirroring=NoMirror
     resourceName=None
     resourceObject=/Game/Athena/HUD/LTM/Wax/Materials/T_UI_WaxCoin
     tiling=NoTile
     imageSize=<32 32>
     margin=<ltrb 0 0 0 0>
     outlineSettings {
      bUseBrushTransparency=false
      roundingType=HalfHeightRadius
      width=0
      color {colorUseRule=UseColor_Specified, specifiedColor=<rgba 0 0 0 0>}
      cornerRadii=<0 0 0 0>
     }
     tintColor {colorUseRule=UseColor_Specified, specifiedColor=<rgba 1 1 1 1>}
     uVRegion {bIsValid=false, max=<0 0>, min=<0 0>}
    }}
   slot {class=HorizontalBoxSlot, properties {verticalAlignment=VAlign_Center}}
  }
  CommonNumericTextBlock_95 {
   widgetClassPath=/Script/CommonUI.CommonNumericTextBlock
   parent=HorizontalBox_1
   widget=CommonNumericTextBlock_95
   properties {
    currentNumericValue=125
    text=Text Block
    colorAndOpacity {colorUseRule=UseColor_Specified, specifiedColor=<rgba 1 0.93 0.4896 1>}
    font {
     bForceMonospaced=false
     bMaterialIsStencil=false
     fontMaterial=null
     fontObject=&Burbank
     letterSpacing=0
     monospacedWidth=1
     size=26
     skewAmount=0
     typefaceFontName=Burbank-Black
     outlineSettings {
      bApplyOutlineToDropShadows=false
      bMiteredCorners=false
      bSeparateFillAlpha=true
      outlineMaterial=null
      outlineSize=1
      outlineColor=<rgba 0.0052 0.0052 0.0056 0.6>
     }
    }
    renderTransform {angle=0, scale=<1 1>, shear=<-12 0>, translation=<0 0>}
    shadowOffset=<0 0>
   }
   slot {class=HorizontalBoxSlot, properties {horizontalAlignment=HAlign_Left, verticalAlignment=VAlign_Center, padding=<ltrb 0 -2 0 0>}}
  }
  Image_2 {
   widgetClassPath=&Image
   bIsVariable=true
   parent=VerticalBoxHeader
   widget=Image_2
   properties {
    brush {
     bIsDynamicallyLoaded=false
     drawAs=Image
     imageType=NoImage
     mirroring=NoMirror
     resourceName=None
     resourceObject=null
     tiling=NoTile
     imageSize=<2 2>
     margin=<ltrb 0 0 0 0>
     outlineSettings {
      bUseBrushTransparency=false
      roundingType=HalfHeightRadius
      width=0
      color {colorUseRule=UseColor_Specified, specifiedColor=<rgba 0 0 0 0>}
      cornerRadii=<0 0 0 0>
     }
     tintColor {colorUseRule=UseColor_Specified, specifiedColor=<rgba 1 1 1 1>}
     uVRegion {bIsValid=false, max=<0 0>, min=<0 0>}
    }
    colorAndOpacity=<rgba 1 1 1 0.1>
   }
   slot {class=VerticalBoxSlot, properties {padding=<ltrb 0 10 0 10>}}
  }
  VerticalBox_73 {
   widgetClassPath=&VerticalBox
   parent=VerticalBox_1
   widget=VerticalBox_73
   slot {class=VerticalBoxSlot, properties {verticalAlignment=VAlign_Top, padding=<ltrb 5 0 5 0>, size {sizeRule=Fill, value=1}}}
  }
  GenericLeaderboardEntry {
   widgetClassPath=&/GenericLeaderboardEntry
   bIsVariable=true
   parent=VerticalBox_73
   widget=GenericLeaderboardEntry
   slot {class=VerticalBoxSlot}
  }
  GenericLeaderboardEntry_C_1 {
   widgetClassPath=&GenericLeaderboardEntry
   bIsVariable=true
   parent=VerticalBox_73
   widget=GenericLeaderboardEntry_C_1
   slot {class=VerticalBoxSlot}
  }
  GenericLeaderboardEntry_C_2 {
   widgetClassPath=&GenericLeaderboardEntry
   bIsVariable=true
   parent=VerticalBox_73
   widget=GenericLeaderboardEntry_C_2
   slot {class=VerticalBoxSlot}
  }
  GenericLeaderboardEntry_C_3 {
   widgetClassPath=&GenericLeaderboardEntry
   bIsVariable=true
   parent=VerticalBox_73
   widget=GenericLeaderboardEntry_C_3
   slot {class=VerticalBoxSlot}
  }
  TutorialHighlight {widgetClassPath=&TutorialHighlight}
 ]
}
