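# Translucent UI material for LTM loading screens that draws a skewed box background with an animated storm-tech grid.
# A two-color linear gradient blends toward Grid Color through MF_UI_BR_StormTech cells, with time driving the random seed.
# Four rotated, smoothstepped UV edges combine into the box mask, minus an optional corner cut.
# UseTimeOffset blends in slight sine-driven wobble on each edge's angle, phase-shifted and at differing rates.
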Material /Game/UI/Foundation/Textures/LoadingScreens/LTM/Materials/M_UI_SkewBox_Background
graph {
 properties {
  bUsedWithStaticMesh=false
  blendMode=BLEND_Translucent
  materialDomain=MD_UI
  pixelDepthOffsetMode=PDOM_Legacy
  usedShadingModels=Default Lit
 }
 nodes [
  n1 = ScalarParameter() {defaultValue=1, parameterName=Rotation, sliderMax=1}
  n2 = MaterialFunctionCall() {materialFunction=&GetUserInterfaceUV}
  LinearInterpolate(
   A=VectorParameter() {parameterName=Color1, defaultValue=<rgba 1 1 1 1>}.A
   B=VectorParameter() {parameterName=Color2, defaultValue=<rgba 1 1 1 1>}.A
   Alpha=ComponentMask(None=MaterialFunctionCall(UVs=@n2.Normalized UV, "Rotation Angle (0-1)"=@n1) {materialFunction=&CustomRotator}.Rotated Values) {r=true}
  )
  n3 = MaterialFunctionCall(Float2=Divide(
    A=MaterialFunctionCall() {materialFunction=&ScreenResolution}.Visible Resolution
    B=MaterialFunctionCall() {materialFunction=&GetUserInterfaceUV}.Pixel Size
   )) {materialFunction=&BreakOutFloat2Components}
  Multiply(A=@n3.G, B=Multiply(A=@n3.R))
  n4 = Constant3Vector() {constant=<rgba 0.3952 0.0145 0.6146 1>}
  n5 = VectorParameter() {group=Backplate, parameterName=Backplate Color, defaultValue=<rgba 0.0696 0 0.2031 1>}
  n6 = Saturate(None=MaterialFunctionCall(GradientEnd=Constant() {r=0.5}) {materialFunction=&MF_UI_LinearGradient}.Result)
  n7 = Constant()
  n8 = Multiply(A=Time(), B=ScalarParameter() {defaultValue=0.25, group=Pattern, parameterName=Grid Speed})
  n9 = Constant2Vector() {g=0.85, r=0.925}
  n10 = Constant()
  n11 = Constant() {r=0.05}
  n12 = ScalarParameter() {group=Pattern, parameterName=Main Reduction}
  n13 = Constant2Vector() {g=0.3}
  n14 = Constant() {r=0.025}
  n15 = Constant2Vector() {g=0.5, r=0.75}
  n16 = Constant()
  n17 = ScalarParameter() {defaultValue=1, group=Pattern, parameterName=Sub Reduction}
  n18 = ScalarParameter() {defaultValue=35, group=Pattern, parameterName=Tiling}
  n19 = Constant2Vector() {g=0.5, r=0.1}
  NamedRerouteDeclaration(None=LinearInterpolate(
    A=LinearInterpolate(A=@n4, B=@n5.RGB, Alpha=@n6)
    B=VectorParameter() {group=Pattern, parameterName=Grid Color, defaultValue=<rgba 0.4793 0.0176 0.7454 1>}.RGB
    Alpha=MaterialFunctionCall!(
     Screen UV
     Tiling=@n18
     Rotation=@n7
     Main Min/Max Opacity=@n19
     Main Min/Max Scale=@n9
     Main Min/Max Perspective=@n10
     Main Outline Thickness
     Main Speed=@n11
     Main Reduction=@n12
     Sub Min/Max Opacity=@n13
     Sub Min/Max Scale=@n15
     Sub Min/Max Perspective=@n16
     Sub Outline Thickness
     Sub Speed=@n14
     Sub Reduction=@n17
     Random Seed=@n8
     Time
    ) {materialFunction=/FortUILibrary/MaterialFunctions/BR/MF_UI_BR_StormTech}.Main Cells
   )) {name=Pattern, nodeColor=<rgba 1 0.5859 0 1>}
  n20 = MaterialFunctionCall() {materialFunction=&GetUserInterfaceUV}
  n21 = ScalarParameter() {parameterName=TopRotation, sliderMax=0.25, sliderMin=-0.25}
  StaticSwitchParameter(
   True=Divide(
    A=ScalarParameter() {parameterName=TopAngleSize}
    B=MaterialFunctionCall(Float2=@n20.Pixel Size) {materialFunction=&BreakOutFloat2Components}.G
   )
   False=@n21
  ) {parameterName=NormalizeAngle}
  n22 = ScalarParameter() {defaultValue=0.2476, parameterName=LeftEdge}
  n23 = AppendVector(A=@n22, B=Constant() {r=0.5})
  n24 = ScalarParameter() {defaultValue=-0.0576, parameterName=LeftRotation, sliderMax=0.25, sliderMin=-0.25}
  n25 = Time()
  n26 = ScalarParameter() {defaultValue=1, parameterName=Speed}
  n27 = Multiply(A=Add(A=Multiply(A=@n25, B=@n26)) {constB=0}) {constB=0.075}
  n28 = Add(A=Multiply(A=Sine(None=@n27)) {constB=0.5}) {constB=0.5}
  n29 = LinearInterpolate(
   A=@n24
   B=LinearInterpolate(A=Add(A=@n24) {constB=-0.0063}, B=Add(A=@n24) {constB=0.0025}, Alpha=Saturate(None=@n28))
   Alpha=ScalarParameter() {parameterName=UseTimeOffset}
  )
  n30 = MaterialFunctionCall(
   Alpha=ComponentMask(None=MaterialFunctionCall(UVs=@n2.9-Slice UV, Rotation Center=@n23, "Rotation Angle (0-1)"=@n29) {materialFunction=&CustomRotator}.Rotated Values) {r=true}
   Min=@n22
   Max=Add(A=@n22, B=ScalarParameter() {defaultValue=0.008, parameterName=U_EdgeSoftness})
  ) {materialFunction=&SmoothStep}
  n31 = OneMinus(None=ScalarParameter() {defaultValue=0.2476, parameterName=RightEdge})
  n32 = Constant() {r=0.5}
  n33 = ScalarParameter() {defaultValue=-0.0576, parameterName=RightRotation, sliderMax=0.25, sliderMin=-0.25}
  n34 = Add(A=Multiply(A=Time(), B=@n26))
  n35 = Multiply(A=Sine(None=Multiply(A=@n34) {constB=0.1})) {constB=0.5}
  n36 = LinearInterpolate(A=Add(A=@n33) {constB=0.0033}, B=Add(A=@n33) {constB=-0.0025}, Alpha=Saturate(None=Add(A=@n35) {constB=0.5}))
  n37 = ScalarParameter() {parameterName=UseTimeOffset}
  n38 = ComponentMask(None=MaterialFunctionCall(UVs=@n2.9-Slice UV, Rotation Center=AppendVector(A=@n31, B=@n32), "Rotation Angle (0-1)"=LinearInterpolate(A=@n33, B=@n36, Alpha=@n37)) {materialFunction=&CustomRotator}.Rotated Values) {r=true}
  n39 = Add(A=@n31, B=ScalarParameter() {defaultValue=0.008, parameterName=U_EdgeSoftness})
  n40 = OneMinus(None=ScalarParameter() {parameterName=BottomEdge})
  n41 = ScalarParameter() {parameterName=BottomRotation, sliderMax=0.25, sliderMin=-0.25}
  n42 = Add(A=@n41) {constB=-0.0082}
  n43 = Add(A=@n41) {constB=0.0012}
  n44 = Multiply(A=Time(), B=@n26)
  n45 = Sine(None=Multiply(A=Add(A=@n44) {constB=0.333}) {constB=0.025})
  n46 = Saturate(None=Add(A=Multiply(A=@n45) {constB=0.5}) {constB=0.5})
  n47 = MaterialFunctionCall(
   UVs=StaticSwitchParameter(True=@n2.Normalized UV, False=@n2.Normalized UV) {parameterName=BottomNormalized}
   Rotation Center=AppendVector(A=@n40, B=Constant() {r=0.5})
   "Rotation Angle (0-1)"=LinearInterpolate(A=@n41, B=LinearInterpolate(A=@n42, B=@n43, Alpha=@n46), Alpha=ScalarParameter() {parameterName=UseTimeOffset})
  ) {materialFunction=&CustomRotator}
  n48 = ScalarParameter() {defaultValue=0.001, parameterName=V_EdgeSoftness}
  n49 = Max(
   A=Max(
    A=OneMinus(None=@n30.Result)
    B=MaterialFunctionCall(Alpha=@n38, Min=@n31, Max=@n39) {materialFunction=&SmoothStep}.Result
   )
   B=MaterialFunctionCall(Alpha=ComponentMask(None=@n47.Rotated Values) {g=true}, Min=@n40, Max=Add(A=@n40, B=@n48)) {materialFunction=&SmoothStep}.Result
  )
  n50 = ScalarParameter() {parameterName=TopEdge}
  n51 = Add(A=@n21) {constB=0.0025}
  n52 = Add(A=@n21) {constB=-0.0025}
  n53 = Multiply(A=Time(), B=@n26)
  n54 = Sine(None=Multiply(A=Add(A=@n53) {constB=0.666}) {constB=0.05})
  n55 = Saturate(None=Add(A=Multiply(A=@n54) {constB=0.5}) {constB=0.5})
  n56 = MaterialFunctionCall(
   UVs=StaticSwitchParameter(True=@n2.Normalized UV, False=@n2.Normalized UV) {parameterName=TopNormalized}
   Rotation Center=AppendVector(A=@n50, B=ScalarParameter() {defaultValue=0.5, parameterName=TopVOffset})
   "Rotation Angle (0-1)"=LinearInterpolate(A=@n21, B=LinearInterpolate(A=@n51, B=@n52, Alpha=@n55), Alpha=ScalarParameter() {parameterName=UseTimeOffset})
  ) {materialFunction=&CustomRotator}
  n57 = ScalarParameter() {defaultValue=0.001, parameterName=V_EdgeSoftness}
  n58 = OneMinus(None=MaterialFunctionCall(Alpha=ComponentMask(None=@n56.Rotated Values) {g=true}, Min=@n50, Max=Add(A=@n50, B=@n57)) {materialFunction=&SmoothStep}.Result)
  n59 = OneMinus(None=ScalarParameter() {defaultValue=0.1, group=Cut, parameterName=CutBottomEdge})
  n60 = MaterialFunctionCall(
   UVs=@n2.9-Slice UV
   Rotation Center=AppendVector(A=@n59, B=Constant() {r=0.5})
   "Rotation Angle (0-1)"=ScalarParameter() {defaultValue=0.0571, group=Cut, parameterName=CutBottomRotation}
  ) {materialFunction=&CustomRotator}
  n61 = ScalarParameter() {defaultValue=0.001, parameterName=Chip_EdgeSoftness}
  n62 = ScalarParameter() {defaultValue=0.1, group=Cut, parameterName=CutLeftEdge}
  n63 = Constant() {r=0.5}
  n64 = ComponentMask(None=MaterialFunctionCall(
    UVs=@n2.9-Slice UV
    Rotation Center=AppendVector(A=@n62, B=@n63)
    "Rotation Angle (0-1)"=ScalarParameter() {group=Cut, parameterName=CutLeftRotation}
   ) {materialFunction=&CustomRotator}.Rotated Values) {r=true}
  n65 = Add(A=@n62, B=ScalarParameter() {defaultValue=0.001, parameterName=Chip_EdgeSoftness})
  n66 = Subtract(
   A=MaterialFunctionCall(Alpha=ComponentMask(None=@n60.Rotated Values) {g=true}, Min=@n59, Max=Add(A=@n59, B=@n61)) {materialFunction=&SmoothStep}.Result
   B=OneMinus(None=MaterialFunctionCall(Alpha=@n64, Min=@n62, Max=@n65) {materialFunction=&SmoothStep}.Result)
  )
  Subtract(A=OneMinus(None=Max(A=@n49, B=@n58)), B=Multiply(A=Clamp(None=@n66), B=ScalarParameter() {group=Cut, parameterName=CutAmount}))
  TextureCoordinate()
  TextureCoordinate()
  TextureCoordinate()
  TextureCoordinate()
  TextureCoordinate()
  TextureCoordinate()
  TextureCoordinate()
 ]
}
