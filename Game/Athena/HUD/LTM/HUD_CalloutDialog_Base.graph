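# HUD widget that shows game mode message callouts in LTMs, with a title, description and mode-info prompt that toggles the fullscreen map.
# Outside replays it binds to FortGameStateAthena message requests, handling only Corner types and queuing any that arrive while one shows.
# Each message's text splits at "(namebreak)" into title and description.
# Intro animation, display-time delay and Outro run in sequence, then the next queued message plays.
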WidgetBlueprint /Game/Athena/HUD/LTM/HUD_CalloutDialog_Base
animations [
 Intro {
  binding_count=4
  length_seconds=1.895
  object=&self:Intro
  sequence {
   bindings [
    Border_Description {
     id=#1
     possessed_class=CommonBorder
     tracks [{
       type=MovieSceneFloatVectorTrack
       display_name=Desired Size Scale
       sections [{
         type=MovieSceneFloatVectorSection
         channels [
          Vector.X {type=MovieSceneScriptingFloatChannel, keys [[17,0],[22.589,1]]}
          Vector.Y {type=MovieSceneScriptingFloatChannel, keys [[20.214,0],[25.803,1]]}
         ]
        }]
      }]
    }
    Overlay_ModeInfo {
     id=#2
     possessed_class=Overlay
     tracks [
      {
       type=MovieSceneFloatTrack
       display_name=Render Opacity
       sections [{type=MovieSceneFloatSection, channels [None {type=MovieSceneScriptingFloatChannel, keys [[30,0],[35,1]]}]}]
      }
      {
       type=MovieScene2DTransformTrack
       sections [{
         type=MovieScene2DTransformSection
         channels [
          Translation.X {type=MovieSceneScriptingFloatChannel, keys [[30,-81.8393],[37.9,33]]}
          Translation.Y {type=MovieSceneScriptingFloatChannel}
          Angle {type=MovieSceneScriptingFloatChannel}
          Scale.X {type=MovieSceneScriptingFloatChannel}
          Scale.Y {type=MovieSceneScriptingFloatChannel}
          Shear.X {type=MovieSceneScriptingFloatChannel}
          Shear.Y {type=MovieSceneScriptingFloatChannel}
         ]
        }]
      }
     ]
    }
    CommonTextBlock_Title1 {
     id=#3
     possessed_class=CommonTextBlock
     tracks [{
       type=MovieSceneFloatTrack
       display_name=Render Opacity
       sections [{type=MovieSceneFloatSection, end_frame=4, start_frame=0, channels [None {type=MovieSceneScriptingFloatChannel, keys [[0,0],[2,0],[4,1]]}]}]
      }]
    }
    Overlay_Container {
     id=#4
     possessed_class=Overlay
     tracks [
      {
       type=MovieScene2DTransformTrack
       sections [{
         type=MovieScene2DTransformSection
         channels [
          Translation.X {type=MovieSceneScriptingFloatChannel}
          Translation.Y {type=MovieSceneScriptingFloatChannel}
          Angle {type=MovieSceneScriptingFloatChannel}
          Scale.X {type=MovieSceneScriptingFloatChannel, keys [[0,2],[7,2],[14.9677,1.0111],[17,1]]}
          Scale.Y {type=MovieSceneScriptingFloatChannel, keys [[7,2],[14.9677,1.0111],[17,1]]}
          Shear.X {type=MovieSceneScriptingFloatChannel}
          Shear.Y {type=MovieSceneScriptingFloatChannel}
         ]
        }]
      }
      {
       type=MovieSceneFloatTrack
       display_name=Render Opacity
       sections [{type=MovieSceneFloatSection, channels [None {type=MovieSceneScriptingFloatChannel, keys [[0,0],[2,1]]}]}]
      }
     ]
    }
   ]
   tracks [{
     type=MovieSceneAudioTrack
     display_name=Audio
     sections [
      {
       type=MovieSceneAudioSection
       end_frame=37
       start_frame=26
       channels [
        Volume {type=MovieSceneScriptingFloatChannel, default=1}
        Pitch {type=MovieSceneScriptingFloatChannel, default=1}
        AttachActor {type=MovieSceneScriptingActorReferenceChannel, default="<Struct 'MovieSceneObjectBindingID' (0x467a50ee580) {}>"}
       ]
      }
      {
       type=MovieSceneAudioSection
       end_frame=30
       start_frame=0
       channels [
        Volume {type=MovieSceneScriptingFloatChannel, default=1}
        Pitch {type=MovieSceneScriptingFloatChannel, default=1}
        AttachActor {type=MovieSceneScriptingActorReferenceChannel, default="<Struct 'MovieSceneObjectBindingID' (0x467a50e9ea0) {}>"}
       ]
      }
      {
       type=MovieSceneAudioSection
       end_frame=34
       start_frame=17
       channels [
        Volume {type=MovieSceneScriptingFloatChannel, default=1}
        Pitch {type=MovieSceneScriptingFloatChannel, default=1}
        AttachActor {type=MovieSceneScriptingActorReferenceChannel, default="<Struct 'MovieSceneObjectBindingID' (0x467a50e0aa0) {}>"}
       ]
      }
     ]
    }]
  }
 }
 Outro {
  binding_count=1
  length_seconds=0.5
  object=&self:Outro
  sequence {bindings [Overlay_Container {
     id=#5
     possessed_class=Overlay
     tracks [
      {
       type=MovieScene2DTransformTrack
       sections [{
         type=MovieScene2DTransformSection
         channels [
          Translation.X {type=MovieSceneScriptingFloatChannel}
          Translation.Y {type=MovieSceneScriptingFloatChannel}
          Angle {type=MovieSceneScriptingFloatChannel}
          Scale.X {type=MovieSceneScriptingFloatChannel, keys [[0,1],[10,0.75]]}
          Scale.Y {type=MovieSceneScriptingFloatChannel, keys [[0,1],[10,0.75]]}
          Shear.X {type=MovieSceneScriptingFloatChannel}
          Shear.Y {type=MovieSceneScriptingFloatChannel}
         ]
        }]
      }
      {
       type=MovieSceneFloatTrack
       display_name=Render Opacity
       sections [{type=MovieSceneFloatSection, channels [None {type=MovieSceneScriptingFloatChannel, keys [[6,1],[10,0]]}]}]
      }
     ]
    }]}
 }
]
graphs [EventGraph {nodes [
   "On Clicked (ModeInfoButton)"() {is=title K2Node_ComponentBoundEvent}
   > ToggleFullscreenMap(self:Fort HUDContext Object Reference=Get FortHUDContext!() -> (ReturnValue:Fort HUDContext Object Reference) {is=title K2Node_GetBlueprintContextEx})
   n1 = DisplayCalloutMessage!() -> (OutputDelegate, then, CalloutMessageData:Athena Game Message Data Structure) {is=title K2Node_CustomEvent}
   > Set CurrentlyShowingMessage(CurrentlyShowingMessage:bool=true) -> (Output_Get:bool)
   n2 = Split(
    SourceString="To String (Text)"(InText=Break Athena Game Message Data!(AthenaGameMessageData:"Athena Game Message Data Structure (by ref)"=@n1.CalloutMessageData) -> (
      MsgType:EAthenaGameMsgType Enum
      MsgText:Text
      MsgSound:Sound Cue Object Reference
      MsgMediaSource:Media Source Object Reference
      MsgDelay:float=~
      bIsTeamBased:bool=~
      TeamIndex:int
      DisplayTime:float=~
      MessageFlags:EAthenaGameMessageDataFlags Enum
      TargetPlayerController:Player Controller Object Reference
     ) {is=title K2Node_BreakStruct}.MsgText)
    InStr=Get NameBreakString() -> (NameBreakString:String)
    SearchDir=FromStart
   )
   > "SetText (Text)"(self=Get CommonTextBlock_Title1() -> (CommonTextBlock_Title1:Common Text Object Reference), InText="To Text (String)"(InString=@n2))
   > SetText(
    self=Get RichText_Description() -> (RichText_Description:Common Rich Text Block Object Reference)
    InText="To Text (String)"(InString=@n2.RightS)
   )
   > SetVisibility(self=Get Overlay_Container() -> (Overlay_Container:Overlay Object Reference), InVisibility=Visible)
   > Set MessageDisplayTime(MessageDisplayTime:double=Break Athena Game Message Data(AthenaGameMessageData=@n1.CalloutMessageData) -> (MsgDelay=~, bIsTeamBased=~, DisplayTime=~).DisplayTime) -> (Output_Get:double)
   n3 = Get Intro()
   > PlayAnimation(InAnimation=@n3)
   > BindToAnimationFinished(Animation=@n3, Delegate=Create Event())
   n4 = OnMessageRequested!() -> (OutputDelegate, then, MsgData:Athena Game Message Data Structure) {is=title K2Node_CustomEvent}
   > n5 = Switch on EAthenaGameMsgType!(
    execute
    Selection:EAthenaGameMsgType Enum=Break Athena Game Message Data(AthenaGameMessageData=@n4.MsgData) -> (MsgDelay=~, bIsTeamBased=~, DisplayTime=~)
   ) -> (
    None:Exec
    DefaultIntro:Exec
    DefaultMessage:Exec
    DefaultCriticalMessage:Exec
    CommIntro:Exec
    CommMessage:Exec
    CommCriticalMessage:Exec
    CornerIntro:Exec
    CornerMessage:Exec
    CornerCriticalMessage:Exec
    RespawnTurningOffWarning:Exec
    RespawnOffWarning:Exec
    CenterMessage:Exec
    CenterImportantMessage:Exec
    CenterCriticalMessage:Exec
   ) {is=title K2Node_SwitchEnum}
   n6 = Branch(
    execute=[@n5.CornerCriticalMessage, @n5.CornerIntro, @n5.CornerMessage]
    Condition=Get CurrentlyShowingMessage() -> (CurrentlyShowingMessage:bool)
   )
   > Add(
    TargetArray:Array of Athena Game Message Data Structures=Get QueuedMessages() -> (QueuedMessages:Array of Athena Game Message Data Structures)
    NewItem:"Athena Game Message Data Structure (by ref)"=@n4.MsgData
   ) {is=title K2Node_CallArrayFunction}
   DisplayCalloutMessage!(execute=@n6.else, self, CalloutMessageData:Athena Game Message Data Structure=@n4.MsgData) -> (then) {is=title K2Node_CallFunction}
   n7 = IntroAnimationFinished() {is=title K2Node_CustomEvent}
   > UnbindFromAnimationFinished!(execute, self:User Widget Object Reference, Animation:Widget Animation Object Reference=Get Intro(), Delegate:Delegate=@n7) -> (then)
   n8 = Get MessageDisplayTime() -> (MessageDisplayTime:double)
   > n9 = Branch(Condition="float > float"(A=@n8, B=0) {is=title K2Node_CallFunction})
   > n10 = Delay(Duration=@n8)
   n11 = Get Outro() -> (Outro:Widget Animation Object Reference)
   PlayAnimation(execute=[@n10.then, @n9.else], InAnimation=@n11)
   > BindToAnimationFinished(Animation=@n11, Delegate=Create Event())
   n12 = OutroAnimationFinished() {is=title K2Node_CustomEvent}
   > UnbindFromAnimationFinished(Animation=Get Outro(), Delegate=@n12)
   > SetVisibility(self=Get Overlay_Container(), InVisibility=Collapsed)
   > Set CurrentlyShowingMessage(CurrentlyShowingMessage=false)
   > PlayQueuedMessages()
   PlayQueuedMessages() {is=title K2Node_CustomEvent}
   n13 = Get QueuedMessages()
   > Branch(Condition=Is Valid Index!(TargetArray:Array of Athena Game Message Data Structures=@n13, IndexToTest:int) -> (ReturnValue:bool))
   > n14 = Set TempMessageData(TempMessageData:Athena Game Message Data Structure="Get (a copy)"(Array:Array of Athena Game Message Data Structures=@n13) -> (Output:Athena Game Message Data Structure)) -> (Output_Get:Athena Game Message Data Structure)
   > Remove Index!(execute, TargetArray:Array of Athena Game Message Data Structures=Get QueuedMessages(), IndexToRemove:int=0) -> (then)
   > DisplayCalloutMessage(CalloutMessageData=@n14)
   Event Construct()
   > Parent: Construct()
   > n15 = Branch(Condition=IsPlayingReplay!() -> (ReturnValue:bool))
   SetVisibility(execute=[@n17.then, @n16.CastFailed, @n15.then], self=Get Overlay_Container(), InVisibility=Collapsed)
   n16 = Cast To FortGameStateAthena!(execute=@n15.else, Object=GetGameState!() -> (ReturnValue:Game State Base Object Reference)) -> (then, CastFailed, AsFort Game State Athena:Fort Game State Athena Object Reference) {is=title K2Node_DynamicCast}
   > n17 = Bind Event to Game Mode Message Request(self:Fort Game State Athena Object Reference=@n16, Delegate=Create Event()) {is=title K2Node_AddDelegate}
  ]}]
variables [
 CurrentlyShowingMessage {type=boolean}
 MessageDisplayTime {type=number}
 ModeInfo_Text {type=string, default="Press <keybind>ToggleFullScreenMap</> for Mode Info"}
 NameBreakString {type=string, default=(namebreak)}
 QueuedMessages {type="array<AthenaGameMessageData>", default []}
 TempMessageData {
  type=AthenaGameMessageData
  default {bIsTeamBased=false, displayTime=5, messageFlags=None, msgDelay=0, msgMediaSource=null, msgSound=null, msgText="", msgType=None, teamIndex=0}
 }
]
widgets {
 info {inheritedWidgetCount=1, parentClass=&FortHUDElementWidget, rootWidgetClass=&Overlay, widgetCount=22}
 widgets [
  Overlay_0 {widgetClassPath=&Overlay, widget=Overlay_0}
  Overlay_Container {
   widgetClassPath=&Overlay
   bIsVariable=true
   parent=Overlay_0
   widget=Overlay_Container
   properties {visibility=HitTestInvisible}
   slot {class=OverlaySlot, properties {horizontalAlignment=HAlign_Center, verticalAlignment=VAlign_Bottom, padding=<ltrb 0 0 0 200>}}
  }
  Overlay_Callout {widgetClassPath=&Overlay, bIsVariable=true, parent=Overlay_Container, widget=Overlay_Callout, slot {class=OverlaySlot}}
  Overlay {widgetClassPath=&Overlay, parent=Overlay_Callout, widget=Overlay, slot {class=OverlaySlot}}
  Image_CalloutBG_Shadow {
   widgetClassPath=&Image
   parent=Overlay
   widget=Image_CalloutBG_Shadow
   properties {
    visibility=HitTestInvisible
    brush {
     bIsDynamicallyLoaded=false
     drawAs=Box
     imageType=NoImage
     mirroring=NoMirror
     resourceName=None
     resourceObject=/Game/Athena/HUD/LTM/Mash/Materials/M_UI_SkewBox_LTMShelf
     tiling=NoTile
     imageSize=<128 64>
     margin=<ltrb 0.5 0 0.5 0>
     outlineSettings {
      bUseBrushTransparency=false
      roundingType=HalfHeightRadius
      width=0
      color {colorUseRule=UseColor_Specified, specifiedColor=<rgba 0 0 0 0>}
      cornerRadii=<0 0 0 0>
     }
     tintColor {colorUseRule=UseColor_Specified, specifiedColor=<rgba 1 1 1 1>}
     uVRegion {bIsValid=false, max=<0 0>, min=<0 0>}
    }
    colorAndOpacity=<rgba 0 0 0 0.3>
    renderTransform {angle=0, scale=<1 1>, shear=<0 0>, translation=<13 13>}
   }
   slot {class=OverlaySlot, properties {horizontalAlignment=HAlign_Fill, verticalAlignment=VAlign_Fill, padding=<ltrb -50 0 -50 0>}}
  }
  Image_CalloutBG_HL1 {
   widgetClassPath=&Image
   parent=Overlay
   widget=Image_CalloutBG_HL1
   properties {
    visibility=HitTestInvisible
    brush {
     bIsDynamicallyLoaded=false
     drawAs=Box
     imageType=NoImage
     mirroring=NoMirror
     resourceName=None
     resourceObject=&M_UI_SkewBox_LTMShelf
     tiling=NoTile
     imageSize=<128 64>
     margin=<ltrb 0.5 0 0.5 0>
     outlineSettings {
      bUseBrushTransparency=false
      roundingType=HalfHeightRadius
      width=0
      color {colorUseRule=UseColor_Specified, specifiedColor=<rgba 0 0 0 0>}
      cornerRadii=<0 0 0 0>
     }
     tintColor {colorUseRule=UseColor_Specified, specifiedColor=<rgba 1 1 1 1>}
     uVRegion {bIsValid=false, max=<0 0>, min=<0 0>}
    }
    colorAndOpacity=<rgba 0.0168 0 0.0382 1>
    renderTransform {angle=0, scale=<1 1>, shear=<0 0>, translation=<5 5>}
   }
   slot {class=OverlaySlot, properties {horizontalAlignment=HAlign_Fill, verticalAlignment=VAlign_Fill, padding=<ltrb -50 0 -50 0>}}
  }
  Image_CalloutBG_1 {
   widgetClassPath=&Image
   parent=Overlay
   widget=Image_CalloutBG_1
   properties {
    visibility=HitTestInvisible
    brush {
     bIsDynamicallyLoaded=false
     drawAs=Box
     imageType=FullColor
     mirroring=NoMirror
     resourceName=None
     resourceObject=/Game/UI/Foundation/Textures/LoadingScreens/LTM/Materials/M_UI_SkewBox_Background
     tiling=NoTile
     imageSize=<128 190>
     margin=<ltrb 0.5 0 0.5 0>
     outlineSettings {
      bUseBrushTransparency=false
      roundingType=HalfHeightRadius
      width=0
      color {colorUseRule=UseColor_Specified, specifiedColor=<rgba 0 0 0 0>}
      cornerRadii=<0 0 0 0>
     }
     tintColor {colorUseRule=UseColor_Specified, specifiedColor=<rgba 1 1 1 1>}
     uVRegion {bIsValid=false, max=<0 0>, min=<0 0>}
    }
   }
   slot {class=OverlaySlot, properties {horizontalAlignment=HAlign_Fill, verticalAlignment=VAlign_Fill, padding=<ltrb -50 0 -50 0>}}
  }
  VerticalBox_0 {
   widgetClassPath=&VerticalBox
   parent=Overlay
   widget=VerticalBox_0
   properties {renderTransform {angle=0, scale=<1 1>, shear=<-15 0>, translation=<0 0>}}
   slot {class=OverlaySlot, properties {horizontalAlignment=HAlign_Fill, padding=<ltrb 30 0 30 0>}}
  }
  Overlay_4 {
   widgetClassPath=&Overlay
   parent=VerticalBox_0
   widget=Overlay_4
   slot {class=VerticalBoxSlot, properties {horizontalAlignment=HAlign_Left}}
  }
  SizeBox_1 {
   widgetClassPath=&SizeBox
   parent=Overlay_4
   widget=SizeBox_1
   properties {bOverride_MaxDesiredWidth=true, maxDesiredWidth=600}
   slot {class=OverlaySlot, properties {horizontalAlignment=HAlign_Center, verticalAlignment=VAlign_Center}}
  }
  CommonTextBlock_Title1 {
   widgetClassPath=&CommonTextBlock
   bIsVariable=true
   parent=SizeBox_1
   widget=CommonTextBlock_Title1
   properties {
    justification=Right
    text=Header
    textTransformPolicy=ToUpper
    font {
     bForceMonospaced=false
     bMaterialIsStencil=false
     fontMaterial=null
     fontObject=&Burbank
     letterSpacing=0
     monospacedWidth=1
     size=36
     skewAmount=0
     typefaceFontName=BurbankRegular-Black
     outlineSettings {
      bApplyOutlineToDropShadows=false
      bMiteredCorners=false
      bSeparateFillAlpha=false
      outlineMaterial=/Game/UI/Foundation/Materials/M-FontShadow
      outlineSize=0
      outlineColor=<rgba 0 0 0 1>
     }
    }
    shadowColorAndOpacity=<rgba 0 0 0 0.2>
    shadowOffset=<5 5>
   }
   slot {class=SizeBoxSlot, properties {horizontalAlignment=HAlign_Right, verticalAlignment=VAlign_Top, padding=<ltrb 0 22 0 0>}}
  }
  Border_Description {
   widgetClassPath=&CommonBorder
   parent=VerticalBox_0
   widget=Border_Description
   properties {clipping=ClipToBounds, padding=<ltrb 0 0 0 30>}
   slot {class=VerticalBoxSlot, properties {horizontalAlignment=HAlign_Left, verticalAlignment=VAlign_Top, size {sizeRule=Fill, value=1}}}
  }
  SizeBox_Description {
   widgetClassPath=&SizeBox
   parent=Border_Description
   widget=SizeBox_Description
   properties {maxDesiredWidth=360, renderTransformPivot=<0.5 1>}
   slot {class=BorderSlot, properties {padding=<ltrb 0 0 0 30>}}
  }
  RichText_Description {
   widgetClassPath=&CommonRichTextBlock
   bIsVariable=true
   parent=SizeBox_Description
   widget=RichText_Description
   properties {
    accessibleBehavior=Auto
    autoWrapText=true
    bCanChildrenBeAccessible=false
    bOverrideDefaultStyle=true
    bOverride_Cursor=true
    minDesiredWidth=640
    mobileTextBlockScale=1.2
    scrollOrientation=Orient_Vertical
    scrollStyle=&TextScrollStyle-Base_C
    text="» <text color=\"fff\">Placeholder</>"
    defaultTextStyleOverride {
     overflowPolicy=Clip
     transformPolicy=None
     colorAndOpacity {colorUseRule=UseColor_Specified, specifiedColor=<rgba 1 0.3813 0 1>}
     font {
      bForceMonospaced=false
      bMaterialIsStencil=false
      fontMaterial=null
      fontObject=&BurbankSmall
      letterSpacing=0
      monospacedWidth=1
      size=24
      skewAmount=0
      typefaceFontName=Black
      outlineSettings {
       bApplyOutlineToDropShadows=false
       bMiteredCorners=false
       bSeparateFillAlpha=false
       outlineMaterial=null
       outlineSize=0
       outlineColor=<rgba 1 1 1 1>
      }
     }
     highlightColor {colorUseRule=UseColor_Specified, specifiedColor=<rgba 0 0 0 1>}
     highlightShape {
      bIsDynamicallyLoaded=false
      drawAs=Image
      imageType=NoImage
      mirroring=NoMirror
      resourceName=None
      resourceObject=null
      tiling=NoTile
      imageSize=<32 32>
      margin=<ltrb 0 0 0 0>
      outlineSettings {
       bUseBrushTransparency=false
       roundingType=HalfHeightRadius
       width=0
       color {colorUseRule=UseColor_Specified, specifiedColor=<rgba 0 0 0 0>}
       cornerRadii=<0 0 0 0>
      }
      tintColor {colorUseRule=UseColor_Specified, specifiedColor=<rgba 1 1 1 1>}
      uVRegion {bIsValid=false, max=<0 0>, min=<0 0>}
     }
     selectedBackgroundColor {colorUseRule=UseColor_Foreground, specifiedColor=<rgba 1 0 1 1>}
     shadowColorAndOpacity=<rgba 0 0 0 1>
     shadowOffset=<0 0>
     strikeBrush {
      bIsDynamicallyLoaded=false
      drawAs=Image
      imageType=NoImage
      mirroring=NoMirror
      resourceName=None
      resourceObject=null
      tiling=NoTile
      imageSize=<32 32>
      margin=<ltrb 0 0 0 0>
      outlineSettings {
       bUseBrushTransparency=false
       roundingType=HalfHeightRadius
       width=0
       color {colorUseRule=UseColor_Specified, specifiedColor=<rgba 0 0 0 0>}
       cornerRadii=<0 0 0 0>
      }
      tintColor {colorUseRule=UseColor_Specified, specifiedColor=<rgba 1 1 1 1>}
      uVRegion {bIsValid=false, max=<0 0>, min=<0 0>}
     }
     underlineBrush {
      bIsDynamicallyLoaded=false
      drawAs=Image
      imageType=NoImage
      mirroring=NoMirror
      resourceName=None
      resourceObject=null
      tiling=NoTile
      imageSize=<32 32>
      margin=<ltrb 0 0 0 0>
      outlineSettings {
       bUseBrushTransparency=false
       roundingType=HalfHeightRadius
       width=0
       color {colorUseRule=UseColor_Specified, specifiedColor=<rgba 0 0 0 0>}
       cornerRadii=<0 0 0 0>
      }
      tintColor {colorUseRule=UseColor_Specified, specifiedColor=<rgba 1 1 1 1>}
      uVRegion {bIsValid=false, max=<0 0>, min=<0 0>}
     }
    }
   }
   slot {class=SizeBoxSlot, properties {verticalAlignment=VAlign_Top, padding=<ltrb 20 0 10 0>}}
  }
  Overlay_ModeInfo {
   widgetClassPath=&Overlay
   parent=Overlay
   widget=Overlay_ModeInfo
   properties {renderTransform {angle=0, scale=<1 1>, shear=<0 0>, translation=<42 30>}}
   slot {class=OverlaySlot, properties {horizontalAlignment=HAlign_Right, verticalAlignment=VAlign_Bottom}}
  }
  Image_ModeInfoBG {
   widgetClassPath=&Image
   parent=Overlay_ModeInfo
   widget=Image_ModeInfoBG
   properties {
    visibility=HitTestInvisible
    brush {
     bIsDynamicallyLoaded=false
     drawAs=Image
     imageType=NoImage
     mirroring=NoMirror
     resourceName=None
     resourceObject=null
     tiling=NoTile
     imageSize=<32 55>
     margin=<ltrb 0 0.1358 0.1 0.2275>
     outlineSettings {
      bUseBrushTransparency=false
      roundingType=HalfHeightRadius
      width=0
      color {colorUseRule=UseColor_Specified, specifiedColor=<rgba 0 0 0 0>}
      cornerRadii=<0 0 0 0>
     }
     tintColor {colorUseRule=UseColor_Specified, specifiedColor=<rgba 1 1 1 1>}
     uVRegion {bIsValid=false, max=<0 0>, min=<0 0>}
    }
    colorAndOpacity=<rgba 1 0.3813 0 1>
    renderTransform {angle=0, scale=<1 1>, shear=<-11 0>, translation=<0 0>}
   }
   slot {class=OverlaySlot, properties {horizontalAlignment=HAlign_Fill, verticalAlignment=VAlign_Fill}}
  }
  HorizontalBox_52 {
   widgetClassPath=&HorizontalBox
   parent=Overlay_ModeInfo
   widget=HorizontalBox_52
   slot {class=OverlaySlot, properties {verticalAlignment=VAlign_Center, padding=<ltrb 20 0 20 0>}}
  }
  GlobalActionKeybindWidget {
   widgetClassPath=/Game/UI/InGame/KeybindWidget/GlobalActionKeybindWidget
   bIsVariable=true
   parent=HorizontalBox_52
   widget=GlobalActionKeybindWidget
   properties {globalAction=ToggleFullScreenMap}
   slot {class=HorizontalBoxSlot, properties {horizontalAlignment=HAlign_Left, verticalAlignment=VAlign_Top, padding=<ltrb 5 5 5 5>}}
  }
  SizeBox_ModeInfo {
   widgetClassPath=&SizeBox
   parent=HorizontalBox_52
   widget=SizeBox_ModeInfo
   properties {visibility=Visible}
   slot {class=HorizontalBoxSlot, properties {horizontalAlignment=HAlign_Center, verticalAlignment=VAlign_Center}}
  }
  RichTextBlock_ModeInfo {
   widgetClassPath=/Script/UMG.RichTextBlock
   bIsVariable=true
   parent=SizeBox_ModeInfo
   widget=RichTextBlock_ModeInfo
   properties {
    bOverrideDefaultStyle=true
    text=" FOR MODE INFO"
    defaultTextStyleOverride {
     overflowPolicy=Clip
     transformPolicy=None
     colorAndOpacity {colorUseRule=UseColor_Specified, specifiedColor=<rgba 1 1 1 1>}
     font {
      bForceMonospaced=false
      bMaterialIsStencil=false
      fontMaterial=null
      fontObject=&BurbankSmall
      letterSpacing=0
      monospacedWidth=1
      size=15.75
      skewAmount=0
      typefaceFontName=Black
      outlineSettings {
       bApplyOutlineToDropShadows=false
       bMiteredCorners=false
       bSeparateFillAlpha=false
       outlineMaterial=null
       outlineSize=0
       outlineColor=<rgba 0 0 0 1>
      }
     }
     highlightColor {colorUseRule=UseColor_Specified, specifiedColor=<rgba 0 0 0 1>}
     highlightShape {
      bIsDynamicallyLoaded=false
      drawAs=Image
      imageType=NoImage
      mirroring=NoMirror
      resourceName=None
      resourceObject=null
      tiling=NoTile
      imageSize=<32 32>
      margin=<ltrb 0 0 0 0>
      outlineSettings {
       bUseBrushTransparency=false
       roundingType=HalfHeightRadius
       width=0
       color {colorUseRule=UseColor_Specified, specifiedColor=<rgba 0 0 0 0>}
       cornerRadii=<0 0 0 0>
      }
      tintColor {colorUseRule=UseColor_Specified, specifiedColor=<rgba 1 1 1 1>}
      uVRegion {bIsValid=false, max=<0 0>, min=<0 0>}
     }
     selectedBackgroundColor {colorUseRule=UseColor_Foreground, specifiedColor=<rgba 1 0 1 1>}
     shadowColorAndOpacity=<rgba 0 0 0 0>
     shadowOffset=<0 0>
     strikeBrush {
      bIsDynamicallyLoaded=false
      drawAs=Image
      imageType=NoImage
      mirroring=NoMirror
      resourceName=None
      resourceObject=null
      tiling=NoTile
      imageSize=<32 32>
      margin=<ltrb 0 0 0 0>
      outlineSettings {
       bUseBrushTransparency=false
       roundingType=HalfHeightRadius
       width=0
       color {colorUseRule=UseColor_Specified, specifiedColor=<rgba 0 0 0 0>}
       cornerRadii=<0 0 0 0>
      }
      tintColor {colorUseRule=UseColor_Specified, specifiedColor=<rgba 1 1 1 1>}
      uVRegion {bIsValid=false, max=<0 0>, min=<0 0>}
     }
     underlineBrush {
      bIsDynamicallyLoaded=false
      drawAs=Image
      imageType=NoImage
      mirroring=NoMirror
      resourceName=None
      resourceObject=null
      tiling=NoTile
      imageSize=<32 32>
      margin=<ltrb 0 0 0 0>
      outlineSettings {
       bUseBrushTransparency=false
       roundingType=HalfHeightRadius
       width=0
       color {colorUseRule=UseColor_Specified, specifiedColor=<rgba 0 0 0 0>}
       cornerRadii=<0 0 0 0>
      }
      tintColor {colorUseRule=UseColor_Specified, specifiedColor=<rgba 1 1 1 1>}
      uVRegion {bIsValid=false, max=<0 0>, min=<0 0>}
     }
    }
    renderTransform {angle=0, scale=<1 1>, shear=<-11 0>, translation=<0 0>}
   }
   slot {class=SizeBoxSlot}
  }
  ModeInfoButton {
   widgetClassPath=&Button
   parent=Overlay_ModeInfo
   widget=ModeInfoButton
   properties {backgroundColor=<rgba 1 1 1 0>, colorAndOpacity=<rgba 1 1 1 0>}
   slot {class=OverlaySlot, properties {horizontalAlignment=HAlign_Fill, verticalAlignment=VAlign_Fill}}
  }
  TutorialHighlight {widgetClassPath=&TutorialHighlight}
 ]
}
