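# UI-domain material for the tutorial video in the Athena LTM HUD, judging by its name and path.
# Its only node is a texture sample of Tut_MediaPlayer_Video, using the External sampler type meant for media textures.
# The graph shows no parameters, other processing or connections from the sample to a material output.
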Material /Game/Athena/HUD/LTM/Tutorial/Tut_MediaPlayer_Video_Mat
graph {
 properties {materialDomain=MD_UI, pixelDepthOffsetMode=PDOM_Legacy, usedShadingModels=Default Lit}
 nodes [TextureSample() {samplerType=SAMPLERTYPE_External, texture=&/Tut_MediaPlayer_Video}]
}
