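# UI sound cue for the check sound of a tutorial text prompt, routed through the Fort_UI sound class.
# A single wave plays through a modulator that holds pitch fixed, behind a short fixed delay at the root.
# The cue lowers its overall volume and carries a prevent-new concurrency rule in its override settings.
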SoundCue /Game/Sounds/Fort_UI_Sounds/Tutorial/Tutorial_TextPrompt_01_Check_Cue
root=Delay_0
nodes [
 {id=Delay_0, children [Modulator_0], properties {delayMax=0.08, delayMin=0.08}}
 {id=Modulator_0, children [WavePlayer_2], properties {pitchMax=1, pitchMin=1}}
 {id=WavePlayer_2, properties {soundWaveAssetPtr=&/Tutorial_TextPromt_Check_01}}
]
properties {
 maxDistance=2097152
 soundClassObject=&Fort_UI
 volumeMultiplier=0.6
 concurrencyOverrides {resolutionRule=PreventNew}
}
