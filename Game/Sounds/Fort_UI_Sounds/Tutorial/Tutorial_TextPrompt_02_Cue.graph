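# Sound cue for the second tutorial text prompt UI sound in the tutorial set.
# A single wave player plays the Tutorial_TextPrompt_02 wave through a modulator whose pitch range is fixed at normal, so pitch does not vary.
# It routes to the Fort_UI sound class, and its concurrency override resolves conflicts by preventing new instances.
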SoundCue /Game/Sounds/Fort_UI_Sounds/Tutorial/Tutorial_TextPrompt_02_Cue
root=Modulator_0
nodes [
 {id=Modulator_0, children [WavePlayer_0], properties {pitchMax=1, pitchMin=1}}
 {id=WavePlayer_0, properties {soundWaveAssetPtr=&/Tutorial_TextPrompt_02}}
]
properties {maxDistance=2097152, soundClassObject=&Fort_UI, concurrencyOverrides {resolutionRule=PreventNew}}
