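# HUD_ContextualTips_Base is a Widget Blueprint that shows contextual tutorial tips as HUD dialogs and world markers.
# UMG animations drive intro, outro, positive checkmark feedback, marker and indicator pop-ins, and controller and mobile close-button hints.
# CanShowMessages admits non-empty text targeted at the owning or no player, outside the EndGame phase.
# Event graph custom events forward repositioning and world-marker outros to the Fort Hud Context Tutorial Tips base.
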WidgetBlueprint /Game/Athena/HUD/LTM/HUD_ContextualTips_Base
animations [
 Outro {
  binding_count=1
  display_label=""
  length_seconds=0.1602
  object=&self:Outro
  sequence {
   bindings [Overlay_Dialog {
     id=#1
     possessed_class=Overlay
     tracks [
      {
       type=MovieScene2DTransformTrack
       sections [{
         type=MovieScene2DTransformSection
         channels [
          Translation.X {type=MovieSceneScriptingFloatChannel}
          Translation.Y {type=MovieSceneScriptingFloatChannel}
          Angle {type=MovieSceneScriptingFloatChannel}
          Scale.X {type=MovieSceneScriptingFloatChannel, keys [[0,1.0346],[6.09,1],[9.612,0]]}
          Scale.Y {type=MovieSceneScriptingFloatChannel, keys [[0,1.0346],[6.09,1],[9.612,0]]}
          Shear.X {type=MovieSceneScriptingFloatChannel}
          Shear.Y {type=MovieSceneScriptingFloatChannel}
         ]
        }]
      }
      {
       type=MovieSceneFloatTrack
       display_name=Render Opacity
       sections [{type=MovieSceneFloatSection, channels [None {type=MovieSceneScriptingFloatChannel, keys [[0,1],[9.612,0]]}]}]
      }
     ]
    }]
   display_rate {numerator=60}
   tracks [{type=MovieSceneAudioTrack, display_name=Audio, sections []}]
  }
 }
 Delay {
  binding_count=0
  display_label=""
  length_seconds=5.0089
  object=&self:Delay
  sequence {bindings [], tracks [{type=MovieSceneEventTrack, sections [{type=MovieSceneEventSection}]}]}
 }
 Feedback_Positive_V2 {
  binding_count=15
  length_seconds=2.0001
  object=&self:Feedback_Positive_V2
  sequence {bindings [
    Checkmark {
     id=#2
     possessed_class=Image
     tracks [{
       type=MovieScene2DTransformTrack
       sections [{
         type=MovieScene2DTransformSection
         channels [
          Translation.X {type=MovieSceneScriptingFloatChannel}
          Translation.Y {type=MovieSceneScriptingFloatChannel}
          Angle {type=MovieSceneScriptingFloatChannel}
          Scale.X {type=MovieSceneScriptingFloatChannel, keys [[0,10],[4,1],[15,1]]}
          Scale.Y {type=MovieSceneScriptingFloatChannel, keys [[0,10],[4,1],[15,1]]}
          Shear.X {type=MovieSceneScriptingFloatChannel}
          Shear.Y {type=MovieSceneScriptingFloatChannel}
         ]
        }]
      }]
    }
    Overlaytriangles {
     id=#3
     possessed_class=Overlay
     tracks [{
       type=MovieSceneFloatTrack
       display_name=Render Opacity
       sections [{type=MovieSceneFloatSection, channels [None {type=MovieSceneScriptingFloatChannel, keys [[0,0],[9.5523,0],[11.9017,1],[40.0007,0]]}]}]
      }]
    }
    Overlay_PositiveFeedback_2 {
     id=#4
     possessed_class=Overlay
     tracks [
      {
       type=MovieScene2DTransformTrack
       sections [{
         type=MovieScene2DTransformSection
         channels [
          Translation.X {type=MovieSceneScriptingFloatChannel}
          Translation.Y {type=MovieSceneScriptingFloatChannel}
          Angle {type=MovieSceneScriptingFloatChannel}
          Scale.X {type=MovieSceneScriptingFloatChannel, keys [[0,0],[5.9133,0],[16.947,4.5043]]}
          Scale.Y {type=MovieSceneScriptingFloatChannel, keys [[0,0],[5.9133,0],[16.947,3.7578]]}
          Shear.X {type=MovieSceneScriptingFloatChannel}
          Shear.Y {type=MovieSceneScriptingFloatChannel}
         ]
        }]
      }
      {
       type=MovieSceneFloatTrack
       display_name=Render Opacity
       sections [{type=MovieSceneFloatSection, channels [None {type=MovieSceneScriptingFloatChannel, keys [[0,0.8],[5.9133,0.8],[16.8953,0]]}]}]
      }
     ]
    }
    Overlay_CheckmarkContainer {
     id=#5
     possessed_class=Overlay
     tracks [{
       type=MovieSceneFloatTrack
       display_name=Render Opacity
       sections [{type=MovieSceneFloatSection, channels [None {type=MovieSceneScriptingFloatChannel, thinned_from=4, keys [[0,0],[4,1],[15,1]]}]}]
      }]
    }
    Overlay_Dialog {
     id=#6
     possessed_class=Overlay
     tracks [
      {
       type=MovieSceneFloatVectorTrack
       display_name=Pivot
       sections [{
         type=MovieSceneFloatVectorSection
         channels [
          Vector.X {type=MovieSceneScriptingFloatChannel, keys [[0,0],[1,0.5]]}
          Vector.Y {type=MovieSceneScriptingFloatChannel, keys [[0,0.6439],[1,0.5]]}
         ]
        }]
      }
      {
       type=MovieScene2DTransformTrack
       sections [{
         type=MovieScene2DTransformSection
         channels [
          Translation.X {type=MovieSceneScriptingFloatChannel}
          Translation.Y {type=MovieSceneScriptingFloatChannel}
          Angle {type=MovieSceneScriptingFloatChannel}
          Scale.X {
           type=MovieSceneScriptingFloatChannel
           keys [[0,1],[3.3423,1],[3.8023,0.8401],[5.3423,1.3471],[6.3423,0.6676],[7.3423,1],[8.2907,0.9351],[9.3423,1],[10.3423,1]]
          }
          Scale.Y {
           type=MovieSceneScriptingFloatChannel
           keys [[0,1],[3.3423,1],[3.8023,0.8401],[5.3113,1.3122],[6.332,0.7477],[7.425,1.0209],[8.3423,0.9134],[9.3423,1],[10.3423,1]]
          }
          Shear.X {type=MovieSceneScriptingFloatChannel}
          Shear.Y {type=MovieSceneScriptingFloatChannel}
         ]
        }]
      }
     ]
    }
    CommonTextBlock_40 {
     id=#7
     possessed_class=CommonTextBlock
     tracks [{
       type=MovieSceneFloatTrack
       display_name=Render Opacity
       sections [{type=MovieSceneFloatSection, channels [None {type=MovieSceneScriptingFloatChannel, keys [[0,1],[3.0073,0]]}]}]
      }]
    }
    OverlayTextContent {
     id=#8
     possessed_class=Overlay
     tracks [{
       type=MovieScene2DTransformTrack
       sections [{
         type=MovieScene2DTransformSection
         channels [
          Translation.X {type=MovieSceneScriptingFloatChannel}
          Translation.Y {type=MovieSceneScriptingFloatChannel}
          Angle {type=MovieSceneScriptingFloatChannel}
          Scale.X {type=MovieSceneScriptingFloatChannel, const [1,3]}
          Scale.Y {type=MovieSceneScriptingFloatChannel, const [1,3]}
          Shear.X {type=MovieSceneScriptingFloatChannel}
          Shear.Y {type=MovieSceneScriptingFloatChannel}
         ]
        }]
      }]
    }
    Overlay_Check {
     id=#9
     display_name="Checkmark (14)"
     possessed_class=Overlay
     tracks [{
       type=MovieScene2DTransformTrack
       sections [{
         type=MovieScene2DTransformSection
         channels [
          Translation.X {type=MovieSceneScriptingFloatChannel}
          Translation.Y {type=MovieSceneScriptingFloatChannel}
          Angle {type=MovieSceneScriptingFloatChannel, const [0,2]}
          Scale.X {type=MovieSceneScriptingFloatChannel, const [1.6,1]}
          Scale.Y {type=MovieSceneScriptingFloatChannel, const [1.6,1]}
          Shear.X {type=MovieSceneScriptingFloatChannel}
          Shear.Y {type=MovieSceneScriptingFloatChannel}
         ]
        }]
      }]
    }
    Image_BGInset_Corner {
     id=#10
     possessed_class=Image
     tracks [
      {
       type=MovieSceneColorTrack
       display_name=Color and Opacity
       sections [{
         type=MovieSceneColorSection
         channels [
          Color.R {type=MovieSceneScriptingFloatChannel, keys [[0,0],[3.7973,0],[7.2027,1]]}
          Color.G {type=MovieSceneScriptingFloatChannel, keys [[0,0.117],[3.7973,0.117],[7.2027,1]]}
          Color.B {type=MovieSceneScriptingFloatChannel, const [1,3]}
          Color.A {type=MovieSceneScriptingFloatChannel, const [1,3]}
         ]
        }]
      }
      {
       type=MovieSceneWidgetMaterialTrack
       sections [{type=MovieSceneParameterSection, channels [LeftEdge {type=MovieSceneScriptingFloatChannel, keys [[0,0.941],[7,0.8313],[18.063,0.861]]}]}]
      }
     ]
    }
    Image1Cross {
     id=#11
     possessed_class=Image
     tracks [
      {
       type=MovieSceneFloatTrack
       display_name=Render Opacity
       sections [{type=MovieSceneFloatSection, channels [None {type=MovieSceneScriptingFloatChannel, keys [[0,0],[3.7973,0],[8.7973,1],[20.1847,0]]}]}]
      }
      {
       type=MovieScene2DTransformTrack
       sections [{
         type=MovieScene2DTransformSection
         channels [
          Translation.X {type=MovieSceneScriptingFloatChannel, const [0,2]}
          Translation.Y {type=MovieSceneScriptingFloatChannel, const [0,2]}
          Scale.X {type=MovieSceneScriptingFloatChannel, keys [[0,2],[3.7973,2],[13.982,2.4527],[26.809,2],[31.797,2]]}
          Scale.Y {type=MovieSceneScriptingFloatChannel, keys [[0,2],[3.7973,2],[13.982,2.0272],[26.809,2],[31.797,2]]}
          Shear.X {type=MovieSceneScriptingFloatChannel, const [0,2]}
          Shear.Y {type=MovieSceneScriptingFloatChannel, const [0,2]}
         ]
        }]
      }
     ]
    }
    Overlay_Container {
     id=#12
     possessed_class=Overlay
     tracks [{
       type=MovieSceneFloatTrack
       display_name=Render Opacity
       sections [{type=MovieSceneFloatSection, end_frame=40, start_frame=0, channels [None {type=MovieSceneScriptingFloatChannel, const [1,2]}]}]
      }]
    }
    Overlay_PositiveFeedback {
     id=#13
     possessed_class=Overlay
     tracks [
      {
       type=MovieScene2DTransformTrack
       sections [{
         type=MovieScene2DTransformSection
         channels [
          Translation.X {type=MovieSceneScriptingFloatChannel}
          Translation.Y {type=MovieSceneScriptingFloatChannel}
          Angle {type=MovieSceneScriptingFloatChannel}
          Scale.X {type=MovieSceneScriptingFloatChannel, keys [[0,0],[3,0],[14.0337,4.5043]]}
          Scale.Y {type=MovieSceneScriptingFloatChannel, keys [[0,0],[3,0],[14.0337,3.7578]]}
          Shear.X {type=MovieSceneScriptingFloatChannel}
          Shear.Y {type=MovieSceneScriptingFloatChannel}
         ]
        }]
      }
      {
       type=MovieSceneFloatTrack
       display_name=Render Opacity
       sections [{type=MovieSceneFloatSection, channels [None {type=MovieSceneScriptingFloatChannel, keys [[0,0.8],[3,0.8],[13.982,0]]}]}]
      }
     ]
    }
    Overlay_PositiveFeedback {
     id=#14
     possessed_class=Overlay
     tracks [
      {
       type=MovieScene2DTransformTrack
       sections [{
         type=MovieScene2DTransformSection
         channels [
          Translation.X {type=MovieSceneScriptingFloatChannel}
          Translation.Y {type=MovieSceneScriptingFloatChannel}
          Angle {type=MovieSceneScriptingFloatChannel}
          Scale.X {type=MovieSceneScriptingFloatChannel, keys [[0,0],[3,0],[14.0337,4.5043]]}
          Scale.Y {type=MovieSceneScriptingFloatChannel, keys [[0,0],[3,0],[14.0337,3.7578]]}
          Shear.X {type=MovieSceneScriptingFloatChannel}
          Shear.Y {type=MovieSceneScriptingFloatChannel}
         ]
        }]
      }
      {
       type=MovieSceneFloatTrack
       display_name=Render Opacity
       sections [{type=MovieSceneFloatSection, channels [None {type=MovieSceneScriptingFloatChannel, keys [[0,0.8],[3,0.8],[13.982,0]]}]}]
      }
     ]
    }
    Image_BGOutline {
     id=#15
     possessed_class=Image
     tracks [{
       type=MovieSceneColorTrack
       display_name=Color and Opacity
       sections [{
         type=MovieSceneColorSection
         channels [
          Color.R {type=MovieSceneScriptingFloatChannel, keys [[0,1],[3.7973,1],[7.2027,0]]}
          Color.G {type=MovieSceneScriptingFloatChannel, const [1,3]}
          Color.B {type=MovieSceneScriptingFloatChannel, keys [[0,1],[3.7973,1],[7.2027,0.2576]]}
          Color.A {type=MovieSceneScriptingFloatChannel, const [1,3]}
         ]
        }]
      }]
    }
    ImageSpinningSpikeball1 {
     id=#16
     possessed_class=Image
     tracks [
      {
       type=MovieSceneFloatTrack
       display_name=Render Opacity
       sections [{
         type=MovieSceneFloatSection
         channels [None {type=MovieSceneScriptingFloatChannel, keys [[0,0],[3.7973,0],[8.7973,1],[31.797,1],[33.7973,0]]}]
        }]
      }
      {
       type=MovieScene2DTransformTrack
       sections [{
         type=MovieScene2DTransformSection
         channels [
          Translation.X {type=MovieSceneScriptingFloatChannel}
          Translation.Y {type=MovieSceneScriptingFloatChannel}
          Angle {type=MovieSceneScriptingFloatChannel, keys [[0,0],[3.7973,0],[33.7973,15]]}
          Scale.X {type=MovieSceneScriptingFloatChannel, keys [[0,2],[7.894,2],[34.7977,1]]}
          Scale.Y {type=MovieSceneScriptingFloatChannel, keys [[0,2],[7.894,2],[34.7977,1]]}
          Shear.X {type=MovieSceneScriptingFloatChannel}
          Shear.Y {type=MovieSceneScriptingFloatChannel}
         ]
        }]
      }
     ]
    }
   ]}
 }
 Intro {
  binding_count=18
  length_seconds=1.4
  object=&self:Intro
  sequence {
   bindings [
    Overlaytriangles {
     id=#3
     possessed_class=Overlay
     tracks [{
       type=MovieSceneFloatTrack
       display_name=Render Opacity
       sections [{type=MovieSceneFloatSection, channels [None {type=MovieSceneScriptingFloatChannel, const [0,2]}]}]
      }]
    }
    Overlay_PositiveFeedback_2 {
     id=#4
     possessed_class=Overlay
     tracks [
      {
       type=MovieScene2DTransformTrack
       sections [{
         type=MovieScene2DTransformSection
         channels [
          Translation.X {type=MovieSceneScriptingFloatChannel}
          Translation.Y {type=MovieSceneScriptingFloatChannel}
          Angle {type=MovieSceneScriptingFloatChannel}
          Scale.X {type=MovieSceneScriptingFloatChannel, const [0,2]}
          Scale.Y {type=MovieSceneScriptingFloatChannel, const [0,2]}
          Shear.X {type=MovieSceneScriptingFloatChannel}
          Shear.Y {type=MovieSceneScriptingFloatChannel}
         ]
        }]
      }
      {
       type=MovieSceneFloatTrack
       display_name=Render Opacity
       sections [{type=MovieSceneFloatSection, channels [None {type=MovieSceneScriptingFloatChannel, const [0.8,2]}]}]
      }
     ]
    }
    Overlay_Dialog {
     id=#6
     possessed_class=Overlay
     tracks [
      {
       type=MovieSceneFloatVectorTrack
       display_name=Pivot
       sections [{
         type=MovieSceneFloatVectorSection
         channels [
          Vector.X {type=MovieSceneScriptingFloatChannel, const [0,2]}
          Vector.Y {type=MovieSceneScriptingFloatChannel, const [0.6439,2]}
         ]
        }]
      }
      {
       type=MovieScene2DTransformTrack
       sections [{
         type=MovieScene2DTransformSection
         channels [
          Translation.X {type=MovieSceneScriptingFloatChannel}
          Translation.Y {type=MovieSceneScriptingFloatChannel}
          Angle {type=MovieSceneScriptingFloatChannel, keys [[0,0],[4,0],[12.012,0.1397],[15.713,0]]}
          Scale.X {type=MovieSceneScriptingFloatChannel, thinned_from=5, keys [[0,0],[9.796,1.1282],[12.436,1],[20,1]]}
          Scale.Y {type=MovieSceneScriptingFloatChannel, thinned_from=5, keys [[0,0],[9.796,1.1282],[12.436,1],[20,1]]}
          Shear.X {type=MovieSceneScriptingFloatChannel}
          Shear.Y {type=MovieSceneScriptingFloatChannel}
         ]
        }]
      }
      {
       type=MovieSceneFloatTrack
       display_name=Render Opacity
       sections [{type=MovieSceneFloatSection, end_frame=84, start_frame=0, channels [None {type=MovieSceneScriptingFloatChannel, const [1,2]}]}]
      }
     ]
    }
    CommonTextBlock_40 {
     id=#7
     possessed_class=CommonTextBlock
     tracks [{
       type=MovieSceneFloatTrack
       display_name=Render Opacity
       sections [{type=MovieSceneFloatSection, channels [None {type=MovieSceneScriptingFloatChannel, const [1,2]}]}]
      }]
    }
    OverlayTextContent {
     id=#8
     possessed_class=Overlay
     tracks [{
       type=MovieScene2DTransformTrack
       sections [{
         type=MovieScene2DTransformSection
         channels [
          Translation.X {type=MovieSceneScriptingFloatChannel}
          Translation.Y {type=MovieSceneScriptingFloatChannel}
          Angle {type=MovieSceneScriptingFloatChannel}
          Scale.X {type=MovieSceneScriptingFloatChannel, const [1,2]}
          Scale.Y {type=MovieSceneScriptingFloatChannel, const [1,2]}
          Shear.X {type=MovieSceneScriptingFloatChannel}
          Shear.Y {type=MovieSceneScriptingFloatChannel}
         ]
        }]
      }]
    }
    SizeBox_Message {
     id=#17
     possessed_class=SizeBox
     tracks [{
       type=MovieScene2DTransformTrack
       sections [{
         type=MovieScene2DTransformSection
         channels [
          Translation.X {type=MovieSceneScriptingFloatChannel}
          Translation.Y {type=MovieSceneScriptingFloatChannel}
          Angle {type=MovieSceneScriptingFloatChannel}
          Scale.X {type=MovieSceneScriptingFloatChannel, keys [[0,1],[14,1],[15.635,1.1],[18.97,1]]}
          Scale.Y {type=MovieSceneScriptingFloatChannel, keys [[0,1],[14,1],[15.635,1.1],[18.97,1]]}
          Shear.X {type=MovieSceneScriptingFloatChannel}
          Shear.Y {type=MovieSceneScriptingFloatChannel}
         ]
        }]
      }]
    }
    Overlay_Container {
     id=#18
     possessed_class=Overlay
     tracks [{
       type=MovieSceneFloatTrack
       display_name=Render Opacity
       sections [{type=MovieSceneFloatSection, channels [None {type=MovieSceneScriptingFloatChannel, keys [[0,0],[3.564,1],[20,1]]}]}]
      }]
    }
    Image_BGInset_Corner {
     id=#10
     possessed_class=Image
     tracks [
      {
       type=MovieSceneColorTrack
       display_name=Color and Opacity
       sections [{
         type=MovieSceneColorSection
         channels [
          Color.R {type=MovieSceneScriptingFloatChannel, const [0,2]}
          Color.G {type=MovieSceneScriptingFloatChannel, const [0.117,2]}
          Color.B {type=MovieSceneScriptingFloatChannel, const [1,2]}
          Color.A {type=MovieSceneScriptingFloatChannel, const [1,2]}
         ]
        }]
      }
      {
       type=MovieSceneWidgetMaterialTrack
       sections [{type=MovieSceneParameterSection, channels [LeftEdge {type=MovieSceneScriptingFloatChannel, keys [[0,1.3004],[10.861,0.941]]}]}]
      }
     ]
    }
    Image1Cross {
     id=#11
     possessed_class=Image
     tracks [
      {
       type=MovieSceneFloatTrack
       display_name=Render Opacity
       sections [{type=MovieSceneFloatSection, channels [None {type=MovieSceneScriptingFloatChannel, const [0,2]}]}]
      }
      {
       type=MovieScene2DTransformTrack
       sections [{
         type=MovieScene2DTransformSection
         channels [
          Translation.X {type=MovieSceneScriptingFloatChannel, const [0,2]}
          Translation.Y {type=MovieSceneScriptingFloatChannel, const [0,2]}
          Scale.X {type=MovieSceneScriptingFloatChannel, const [2,2]}
          Scale.Y {type=MovieSceneScriptingFloatChannel, const [2,2]}
          Shear.X {type=MovieSceneScriptingFloatChannel, const [0,2]}
          Shear.Y {type=MovieSceneScriptingFloatChannel, const [0,2]}
         ]
        }]
      }
     ]
    }
    Overlay_PositiveFeedback {
     id=#13
     possessed_class=Overlay
     tracks [
      {
       type=MovieScene2DTransformTrack
       sections [{
         type=MovieScene2DTransformSection
         channels [
          Translation.X {type=MovieSceneScriptingFloatChannel}
          Translation.Y {type=MovieSceneScriptingFloatChannel}
          Angle {type=MovieSceneScriptingFloatChannel}
          Scale.X {type=MovieSceneScriptingFloatChannel, const [0,2]}
          Scale.Y {type=MovieSceneScriptingFloatChannel, const [0,2]}
          Shear.X {type=MovieSceneScriptingFloatChannel}
          Shear.Y {type=MovieSceneScriptingFloatChannel}
         ]
        }]
      }
      {
       type=MovieSceneFloatTrack
       display_name=Render Opacity
       sections [{type=MovieSceneFloatSection, channels [None {type=MovieSceneScriptingFloatChannel, const [0.8,2]}]}]
      }
     ]
    }
    Overlay_PositiveFeedback {
     id=#14
     possessed_class=Overlay
     tracks [
      {
       type=MovieScene2DTransformTrack
       sections [{
         type=MovieScene2DTransformSection
         channels [
          Translation.X {type=MovieSceneScriptingFloatChannel}
          Translation.Y {type=MovieSceneScriptingFloatChannel}
          Angle {type=MovieSceneScriptingFloatChannel}
          Scale.X {type=MovieSceneScriptingFloatChannel, const [0,2]}
          Scale.Y {type=MovieSceneScriptingFloatChannel, const [0,2]}
          Shear.X {type=MovieSceneScriptingFloatChannel}
          Shear.Y {type=MovieSceneScriptingFloatChannel}
         ]
        }]
      }
      {
       type=MovieSceneFloatTrack
       display_name=Render Opacity
       sections [{type=MovieSceneFloatSection, channels [None {type=MovieSceneScriptingFloatChannel, const [0.8,2]}]}]
      }
     ]
    }
    Overlay_CheckmarkContainer {
     id=#19
     possessed_class=Overlay
     tracks [{
       type=MovieSceneFloatTrack
       display_name=Render Opacity
       sections [{type=MovieSceneFloatSection, channels [None {type=MovieSceneScriptingFloatChannel, const [0,1]}]}]
      }]
    }
    Image_BGOutline {
     id=#15
     possessed_class=Image
     tracks [{
       type=MovieSceneColorTrack
       display_name=Color and Opacity
       sections [{
         type=MovieSceneColorSection
         channels [
          Color.R {type=MovieSceneScriptingFloatChannel, const [1,2]}
          Color.G {type=MovieSceneScriptingFloatChannel, const [1,2]}
          Color.B {type=MovieSceneScriptingFloatChannel, const [1,2]}
          Color.A {type=MovieSceneScriptingFloatChannel, const [1,2]}
         ]
        }]
      }]
    }
    ImageSpinningSpikeball1 {
     id=#16
     possessed_class=Image
     tracks [
      {
       type=MovieSceneFloatTrack
       display_name=Render Opacity
       sections [{type=MovieSceneFloatSection, channels [None {type=MovieSceneScriptingFloatChannel, const [0,2]}]}]
      }
      {
       type=MovieScene2DTransformTrack
       sections [{
         type=MovieScene2DTransformSection
         channels [
          Translation.X {type=MovieSceneScriptingFloatChannel}
          Translation.Y {type=MovieSceneScriptingFloatChannel}
          Angle {type=MovieSceneScriptingFloatChannel, const [0,2]}
          Scale.X {type=MovieSceneScriptingFloatChannel, const [2,2]}
          Scale.Y {type=MovieSceneScriptingFloatChannel, const [2,2]}
          Shear.X {type=MovieSceneScriptingFloatChannel}
          Shear.Y {type=MovieSceneScriptingFloatChannel}
         ]
        }]
      }
     ]
    }
   ]
   display_rate {numerator=60}
  }
 }
 Indicator_In {
  binding_count=3
  length_seconds=0.45
  object=&self:Indicator_In
  sequence {bindings [
    BallContextualBall {
     id=#20
     possessed_class=ContextTutorialBall_C
     tracks [{
       type=MovieSceneByteTrack
       display_name=Visibility
       sections [{type=MovieSceneByteSection, end_frame=7, start_frame=0, channels [None {type=MovieSceneScriptingByteChannel, keys [[0,4],[7,2]]}]}]
      }]
    }
    BallOverlay {
     id=#21
     possessed_class=Overlay
     tracks [
      {
       type=MovieScene2DTransformTrack
       sections [{
         type=MovieScene2DTransformSection
         end_frame=7
         start_frame=0
         channels [
          Translation.X {type=MovieSceneScriptingFloatChannel}
          Translation.Y {type=MovieSceneScriptingFloatChannel}
          Angle {type=MovieSceneScriptingFloatChannel}
          Scale.X {type=MovieSceneScriptingFloatChannel, keys [[2,1],[5,3]]}
          Scale.Y {type=MovieSceneScriptingFloatChannel, keys [[2,1],[5,3]]}
          Shear.X {type=MovieSceneScriptingFloatChannel, const [0,1]}
          Shear.Y {type=MovieSceneScriptingFloatChannel, const [0,1]}
         ]
        }]
      }
      {
       type=MovieSceneFloatTrack
       display_name=Render Opacity
       sections [{type=MovieSceneFloatSection, end_frame=7, start_frame=1, channels [None {type=MovieSceneScriptingFloatChannel, keys [[1,0],[2,1],[6,0]]}]}]
      }
      {
       type=MovieSceneByteTrack
       display_name=Visibility
       sections [{type=MovieSceneByteSection, end_frame=7, start_frame=0, channels [None {type=MovieSceneScriptingByteChannel, keys [[0,4],[7,2]]}]}]
      }
     ]
    }
    "\"!\"_Overlay" {
     id=#22
     possessed_class=Overlay
     tracks [{
       type=MovieScene2DTransformTrack
       sections [{
         type=MovieScene2DTransformSection
         end_frame=9
         start_frame=-1
         channels [
          Translation.X {type=MovieSceneScriptingFloatChannel}
          Translation.Y {type=MovieSceneScriptingFloatChannel}
          Angle {type=MovieSceneScriptingFloatChannel}
          Scale.X {type=MovieSceneScriptingFloatChannel, keys [[-0.0007,1],[1.9993,0.47],[2,0.5],[3,0.6074],[5.0003,1.1],[7,1]]}
          Scale.Y {type=MovieSceneScriptingFloatChannel, keys [[-0.0007,1],[1.9993,1.5608],[1.9997,2],[2,2.5],[3,2],[5.0003,0.95],[7,1]]}
          Shear.X {type=MovieSceneScriptingFloatChannel, const [0,2]}
          Shear.Y {type=MovieSceneScriptingFloatChannel, const [0,2]}
         ]
        }]
      }]
    }
   ]}
 }
 WorldMarker_intro {
  binding_count=11
  length_seconds=0.8
  object=&self:WorldMarker_intro
  sequence {
   bindings [
    HUD_ContextualTips_Base_C_54 {
     id=#23
     possessed_class=HUD_ContextualTips_Base_C
     tracks [
      {
       type=MovieSceneByteTrack
       display_name=Visibility
       sections [{type=MovieSceneByteSection, end_frame=47, start_frame=0, channels [None {type=MovieSceneScriptingByteChannel, const [4,1]}]}]
      }
      {
       type=MovieSceneFloatTrack
       display_name=Render Opacity
       sections [{type=MovieSceneFloatSection, end_frame=0, start_frame=0, channels [None {type=MovieSceneScriptingFloatChannel, const [1,1]}]}]
      }
     ]
    }
    Overlay_Container {
     id=#24
     possessed_class=Overlay
     tracks [{
       type=MovieSceneFloatTrack
       display_name=Render Opacity
       sections [{type=MovieSceneFloatSection, end_frame=0, start_frame=0, channels [None {type=MovieSceneScriptingFloatChannel, const [1,1]}]}]
      }]
    }
    Image_BGInset {
     id=#25
     possessed_class=Image
     tracks [{
       type=MovieSceneByteTrack
       display_name=Visibility
       sections [{type=MovieSceneByteSection, end_frame=47, start_frame=0, channels [None {type=MovieSceneScriptingByteChannel, keys [[0,0],[0,4]]}]}]
      }]
    }
    Image_BGInset_Corner {
     id=#26
     possessed_class=Image
     tracks [{
       type=MovieSceneByteTrack
       display_name=Visibility
       sections [{type=MovieSceneByteSection, end_frame=48, start_frame=0, channels [None {type=MovieSceneScriptingByteChannel, const [4,1]}]}]
      }]
    }
    ScreenOverlay {
     id=#27
     possessed_class=Overlay
     tracks [{
       type=MovieSceneFloatTrack
       display_name=Render Opacity
       sections [{type=MovieSceneFloatSection, end_frame=0, start_frame=0, channels [None {type=MovieSceneScriptingFloatChannel, const [1,1]}]}]
      }]
    }
    ContextTutorialBall {
     id=#28
     possessed_class=ContextTutorialBall_C
     tracks [
      {
       type=MovieSceneByteTrack
       display_name=Visibility
       sections [{
         type=MovieSceneByteSection
         end_frame=48
         start_frame=0
         channels [None {type=MovieSceneScriptingByteChannel, keys [[0,4],[11.368,1],[28.42,1]]}]
        }]
      }
      {
       type=MovieSceneFloatTrack
       display_name=Render Opacity
       sections [{type=MovieSceneFloatSection, end_frame=47, start_frame=0, channels [None {type=MovieSceneScriptingFloatChannel, keys [[7,1],[12,0]]}]}]
      }
      {
       type=MovieScene2DTransformTrack
       sections [{
         type=MovieScene2DTransformSection
         end_frame=15
         start_frame=0
         channels [
          Translation.X {type=MovieSceneScriptingFloatChannel}
          Translation.Y {type=MovieSceneScriptingFloatChannel}
          Angle {type=MovieSceneScriptingFloatChannel}
          Scale.X {type=MovieSceneScriptingFloatChannel, keys [[0,1],[7,0.5],[12,0],[15,1]]}
          Scale.Y {type=MovieSceneScriptingFloatChannel, keys [[0,1],[7,0.5],[12,0],[15,1]]}
          Shear.X {type=MovieSceneScriptingFloatChannel}
          Shear.Y {type=MovieSceneScriptingFloatChannel}
         ]
        }]
      }
     ]
    }
    CommonTextBlock_40 {
     id=#29
     possessed_class=CommonTextBlock
     tracks [
      {
       type=MovieScene2DTransformTrack
       sections [{
         type=MovieScene2DTransformSection
         end_frame=47
         start_frame=0
         channels [
          Scale.X {type=MovieSceneScriptingFloatChannel, keys [[0,0],[9.472,1.2],[16.577,0.6],[21.313,1.4],[23.681,1.25]]}
          Scale.Y {type=MovieSceneScriptingFloatChannel, keys [[0,0],[9.472,1.2],[16.577,2],[21.313,0.8],[23.681,1.25]]}
         ]
        }]
      }
      {
       type=MovieSceneByteTrack
       display_name=Visibility
       sections [{type=MovieSceneByteSection, end_frame=47, start_frame=0, channels [None {type=MovieSceneScriptingByteChannel, const [3,1]}]}]
      }
      {
       type=MovieSceneFloatTrack
       display_name=Render Opacity
       sections [{type=MovieSceneFloatSection, end_frame=47, start_frame=0, channels [None {type=MovieSceneScriptingFloatChannel, const [1,1]}]}]
      }
     ]
    }
    Overlay_Dialog {
     id=#30
     possessed_class=Overlay
     tracks [
      {
       type=MovieSceneByteTrack
       display_name=Visibility
       sections [{type=MovieSceneByteSection, end_frame=0, start_frame=0, channels [None {type=MovieSceneScriptingByteChannel, const [4,1]}]}]
      }
      {
       type=MovieSceneFloatTrack
       display_name=Render Opacity
       sections [{type=MovieSceneFloatSection, end_frame=0, start_frame=0, channels [None {type=MovieSceneScriptingFloatChannel, const [1,1]}]}]
      }
      {
       type=MovieScene2DTransformTrack
       sections [{
         type=MovieScene2DTransformSection
         end_frame=0
         start_frame=0
         channels [
          Translation.X {type=MovieSceneScriptingFloatChannel}
          Translation.Y {type=MovieSceneScriptingFloatChannel}
          Angle {type=MovieSceneScriptingFloatChannel}
          Scale.X {type=MovieSceneScriptingFloatChannel, const [1,1]}
          Scale.Y {type=MovieSceneScriptingFloatChannel, const [1,1]}
          Shear.X {type=MovieSceneScriptingFloatChannel}
          Shear.Y {type=MovieSceneScriptingFloatChannel}
         ]
        }]
      }
     ]
    }
    BorderText {
     id=#31
     possessed_class=Border
     tracks [
      {
       type=MovieScene2DTransformTrack
       sections [{
         type=MovieScene2DTransformSection
         end_frame=47
         start_frame=0
         channels [
          Scale.X {type=MovieSceneScriptingFloatChannel, keys [[0,0.91],[15.779,0.9],[21.313,1.2],[25.343,0.9],[27.466,1]]}
          Scale.Y {type=MovieSceneScriptingFloatChannel, keys [[0,0.91],[15.779,0.9],[21.313,1.2],[25.343,0.9],[27.466,1]]}
         ]
        }]
      }
      {
       type=MovieSceneFloatTrack
       display_name=Render Opacity
       sections [{
         type=MovieSceneFloatSection
         end_frame=47
         start_frame=0
         channels [None {type=MovieSceneScriptingFloatChannel, keys [[13.32,0],[15.779,0],[27.466,1]]}]
        }]
      }
      {
       type=MovieSceneByteTrack
       display_name=Visibility
       sections [{type=MovieSceneByteSection, end_frame=47, start_frame=0, channels [None {type=MovieSceneScriptingByteChannel, const [4,1]}]}]
      }
     ]
    }
    OverlayBG {
     id=#32
     possessed_class=Overlay
     tracks [
      {
       type=MovieSceneFloatVectorTrack
       display_name=Pivot
       sections [{
         type=MovieSceneFloatVectorSection
         end_frame=47
         start_frame=0
         channels [
          Vector.X {type=MovieSceneScriptingFloatChannel, const [1,1]}
          Vector.Y {type=MovieSceneScriptingFloatChannel, const [0,1]}
         ]
        }]
      }
      {
       type=MovieScene2DTransformTrack
       sections [{
         type=MovieScene2DTransformSection
         end_frame=47
         start_frame=0
         channels [
          Scale.X {type=MovieSceneScriptingFloatChannel, keys [[8,0],[14.89,1.4],[20.424,0.9],[24.454,1.05],[26.577,1]]}
          Scale.Y {type=MovieSceneScriptingFloatChannel, keys [[8,0],[14.89,1.4],[20.424,0.9],[24.454,1.05],[26.577,1]]}
         ]
        }]
      }
      {
       type=MovieSceneFloatTrack
       display_name=Render Opacity
       sections [{
         type=MovieSceneFloatSection
         end_frame=47
         start_frame=0
         channels [None {type=MovieSceneScriptingFloatChannel, keys [[10.104,0],[17.683,1]]}]
        }]
      }
      {
       type=MovieSceneByteTrack
       display_name=Visibility
       sections [{type=MovieSceneByteSection, end_frame=47, start_frame=0, channels [None {type=MovieSceneScriptingByteChannel, const [4,1]}]}]
      }
     ]
    }
    Image_BGInset_Corner_Solo {
     id=#33
     possessed_class=Image
     tracks [
      {
       type=MovieScene2DTransformTrack
       sections [{
         type=MovieScene2DTransformSection
         end_frame=48
         start_frame=0
         channels [
          Translation.X {type=MovieSceneScriptingFloatChannel, const [0,2]}
          Translation.Y {type=MovieSceneScriptingFloatChannel, const [0,2]}
          Scale.X {type=MovieSceneScriptingFloatChannel, keys [[0,0],[8.05,2],[15,1.5],[20.315,0.9],[22.683,1]]}
          Scale.Y {type=MovieSceneScriptingFloatChannel, keys [[0,0],[8.05,2],[15,1.5],[20.315,0.9],[22.683,1]]}
         ]
        }]
      }
      {
       type=MovieSceneFloatVectorTrack
       display_name=Pivot
       sections [{
         type=MovieSceneFloatVectorSection
         end_frame=48
         start_frame=0
         channels [
          Vector.X {type=MovieSceneScriptingFloatChannel, keys [[0,0.9953],[20.315,0.9953]]}
          Vector.Y {type=MovieSceneScriptingFloatChannel, const [0.05,1]}
         ]
        }]
      }
      {
       type=MovieSceneFloatTrack
       display_name=Render Opacity
       sections [{type=MovieSceneFloatSection, end_frame=47, start_frame=0, channels [None {type=MovieSceneScriptingFloatChannel, keys [[0,0],[7.104,1]]}]}]
      }
      {
       type=MovieSceneByteTrack
       display_name=Visibility
       sections [{type=MovieSceneByteSection, end_frame=0, start_frame=0, channels [None {type=MovieSceneScriptingByteChannel, const [4,1]}]}]
      }
     ]
    }
   ]
   display_rate {numerator=60}
  }
 }
 MobileCloseButton_Tip_In {
  binding_count=1
  length_seconds=0.2
  object=&self:MobileCloseButton_Tip_In
  sequence {bindings [MobileCloseBUtton {
     id=#34
     possessed_class=Overlay
     tracks [
      {
       type=MovieSceneFloatTrack
       display_name=Render Opacity
       sections [{
         type=MovieSceneFloatSection
         end_frame=4
         start_frame=0
         channels [None {type=MovieSceneScriptingFloatChannel, keys [[0,0],[0.2233,1],[3.9953,1]]}]
        }]
      }
      {
       type=MovieScene2DTransformTrack
       sections [{
         type=MovieScene2DTransformSection
         end_frame=4
         start_frame=0
         channels [
          Translation.X {type=MovieSceneScriptingFloatChannel, keys [[0,-21.1789],[3.0007,62.2633]]}
          Translation.Y {type=MovieSceneScriptingFloatChannel, const [2,2]}
          Scale.X {type=MovieSceneScriptingFloatChannel, keys [[3.0007,0.9],[3.9953,1]]}
          Scale.Y {type=MovieSceneScriptingFloatChannel, keys [[3.0007,0.8],[3.9953,1]]}
         ]
        }]
      }
     ]
    }]}
 }
 Controller_Tip_In {
  binding_count=1
  length_seconds=0.3
  object=&self:Controller_Tip_In
  sequence {bindings [TipForController {
     id=#35
     possessed_class=Overlay
     tracks [
      {
       type=MovieScene2DTransformTrack
       sections [{
         type=MovieScene2DTransformSection
         end_frame=6
         start_frame=0
         channels [
          Translation.X {type=MovieSceneScriptingFloatChannel, keys [[0,-35],[6,-20]]}
          Translation.Y {type=MovieSceneScriptingFloatChannel, const [90,1]}
          Angle {type=MovieSceneScriptingFloatChannel, const [0,1]}
          Scale.X {type=MovieSceneScriptingFloatChannel, const [1,1]}
          Scale.Y {type=MovieSceneScriptingFloatChannel, const [1,1]}
          Shear.X {type=MovieSceneScriptingFloatChannel, const [0,1]}
          Shear.Y {type=MovieSceneScriptingFloatChannel, const [0,1]}
         ]
        }]
      }
      {
       type=MovieSceneFloatTrack
       display_name=Render Opacity
       sections [{type=MovieSceneFloatSection, end_frame=6, start_frame=0, channels [None {type=MovieSceneScriptingFloatChannel, keys [[0,0],[6,1]]}]}]
      }
     ]
    }]}
 }
 Controller_Tip_Out {
  binding_count=1
  length_seconds=0.3
  object=&self:Controller_Tip_Out
  sequence {bindings [TipForController {
     id=#35
     possessed_class=Overlay
     tracks [
      {
       type=MovieScene2DTransformTrack
       sections [{
         type=MovieScene2DTransformSection
         end_frame=6
         start_frame=0
         channels [
          Translation.X {type=MovieSceneScriptingFloatChannel, const [-20,1]}
          Translation.Y {type=MovieSceneScriptingFloatChannel, const [83.8844,1]}
          Angle {type=MovieSceneScriptingFloatChannel, const [0,1]}
          Scale.X {type=MovieSceneScriptingFloatChannel, const [1,1]}
          Scale.Y {type=MovieSceneScriptingFloatChannel, const [1,1]}
          Shear.X {type=MovieSceneScriptingFloatChannel, const [0,1]}
          Shear.Y {type=MovieSceneScriptingFloatChannel, const [0,1]}
         ]
        }]
      }
      {
       type=MovieSceneFloatTrack
       display_name=Render Opacity
       sections [{type=MovieSceneFloatSection, end_frame=6, start_frame=0, channels [None {type=MovieSceneScriptingFloatChannel, keys [[0,1],[6,0]]}]}]
      }
     ]
    }]}
 }
]
graphs [
 CanShowMessages {nodes [
   n1 = CanShowMessages!() -> (then, MessageData:"Athena Game Message Data Structure (by ref)") {is=title K2Node_FunctionEntry}
   n2 = Break Athena Game Message Data!(AthenaGameMessageData:"Athena Game Message Data Structure (by ref)"=@n1) -> (
    MsgType:EAthenaGameMsgType Enum
    MsgText:Text
    MsgSound:Sound Cue Object Reference
    MsgMediaSource:Media Source Object Reference
    MsgDelay:float=~
    bIsTeamBased:bool=~
    TeamIndex:int
    DisplayTime:float=~
    MessageFlags:EAthenaGameMessageDataFlags Enum
    TargetPlayerController:Player Controller Object Reference
   ) {is=title K2Node_BreakStruct}
   n3 = TextIsEmpty(InText=@n2.MsgText)
   n4 = "Equal (Object)"(A=@n2.TargetPlayerController, B=GetOwningPlayer()) {is=title K2Node_CallFunction}
   n5 = IsNullItem!(ItemObject:Object Reference=@n2.TargetPlayerController) -> (Is Null Item:bool)
   n6 = Get CurrentGamePhase() -> (CurrentGamePhase:EAthenaGamePhase Enum=None)
   > Return Node!(
    execute
    bCanShowMessages:bool=AND Boolean!(
     A=NOT Boolean(A=@n3)
     B=OR Boolean(A=@n5, B=@n4)
     C:bool="Not Equal (Enum)"!(A:EAthenaGamePhase Enum=@n6, B:EAthenaGamePhase Enum=EndGame) -> (ReturnValue:bool)
    ) -> (ReturnValue)
   )
  ]}
 EventGraph {nodes [
   n1 = HideMarkerAnimation!() -> (OutputDelegate, then, PlayerID:int, InstanceID:int) {is=title K2Node_CustomEvent}
   > StartWorldMarkerOutroAnimation!(execute, self:Fort Hud Context Tutorial Tips Base Object Reference, InPlayerID:int=@n1.PlayerID, InInstanceID:int=@n1.InstanceID) -> (then)
   n2 = Reposition!() -> (
    OutputDelegate
    then
    Position:Vector 2D Structure
    HorizontalAlignmnent:EHorizontalAlignment Enum=HAlign_Fill
    VerticalAlignment:EVerticalAlignment Enum=VAlign_Fill
   ) {is=title K2Node_CustomEvent}
   > SetPosition!(
    execute
    self:Fort Hud Context Tutorial Tips Base Object Reference
    Position:Vector 2D Structure=@n2.Position
    HorizontalAlignment:EHorizontalAlignment Enum=@n2.HorizontalAlignmnent
    VerticalAlignment:EVerticalAlignment Enum=@n2.VerticalAlignment
   ) -> (then)
   SetDefaultPosition() {is=title K2Node_CustomEvent}
   > SetDefaultPlatformPosition(self:Fort Hud Context Tutorial Tips Base Object Reference)
   n3 = ShowMarkerAnimation!() -> (OutputDelegate, then, PlayerID:int, InstanceID:int) {is=title K2Node_CustomEvent}
   > StartLerpContextTutorialBall!(
    execute
    self:Fort Hud Context Tutorial Tips Base Object Reference
    InPlayerID:int=@n3.PlayerID
    InInstanceID:int=@n3.InstanceID
    TotalTime:float=0.4
    StartDelay:float=0.5
   ) -> (then)
   > PlayAnimation(InAnimation=Get Delay() -> (Delay:Widget Animation Object Reference))
   n4 = OnGamePhaseChanged!() -> (OutputDelegate, then, GamePhase:EAthenaGamePhase Enum=None) {is=title K2Node_CustomEvent}
   > n5 = Set CurrentGamePhase(CurrentGamePhase:EAthenaGamePhase Enum=@n4.GamePhase) -> (Output_Get:EAthenaGamePhase Enum=None)
   > n6 = Switch on EAthenaGamePhase!(execute, Selection:EAthenaGamePhase Enum=@n5) -> (None:Exec, Setup:Exec, Warmup:Exec, Aircraft:Exec, SafeZones:Exec, EndGame:Exec, Count:Exec) {is=title K2Node_SwitchEnum}
   SetVisibility(execute=@n6.EndGame, InVisibility=Collapsed)
   OnHideMessage!() -> (OutputDelegate, then, MEssageType:EAthenaGameMsgType Enum=None, InstantBide:bool=~, ClearQueue:bool=~, MessageChannel:int) {is=title K2Node_CustomEvent}
   > PlayOutro!(execute, self, StopAnimation:bool=true) -> (then)
   n7 = PlaySuccessAnimation!() -> (OutputDelegate, then, OwnerController:Fort Player Controller Object Reference) {is=title K2Node_CustomEvent}
   > n8 = IsOwner!(execute, self, TargetOwner:Fort Player Controller Object Reference=@n7.OwnerController) -> (then, IsOwner:bool)
   > Branch(Condition=@n8)
   > StopAllTipAnimations()
   > PlayAnimation(InAnimation=Get Feedback_Positive_V2() -> (Feedback_Positive_V2:Widget Animation Object Reference))
   > PlaySound2D(Sound=/Game/Sounds/Fort_UI_Sounds/Tutorial/Tutorial_TextPrompt_01_Check_Cue)
   n9 = OnMessageRequested!() -> (OutputDelegate, then, MsgData:Athena Game Message Data Structure, OwnedController:Fort Contextual Tutorial Controller Object Reference) {is=title K2Node_CustomEvent}
   > n10 = CanShowMessages!(execute, self, MessageData:"Athena Game Message Data Structure (by ref)"=@n9.MsgData) -> (then, bCanShowMessages:bool) {is=title K2Node_CallFunction}
   > Branch(Condition=@n10)
   > SetHasActiveContent!(execute, self:Fort Prioritized Widget Interface, bNewActive:bool=true) -> (then)
   > Clear and Invalidate Timer by Handle(Handle=Get OutroAnimationTimerHandle() -> (OutroAnimationTimerHandle:Timer Handle Structure))
   n11 = Break Athena Game Message Data(AthenaGameMessageData=@n9.MsgData) -> (MsgDelay=~, bIsTeamBased=~, DisplayTime=~)
   > n12 = Switch on EAthenaGameMsgType!(execute, Selection:EAthenaGameMsgType Enum=@n11) -> (
    None:Exec
    DefaultIntro:Exec
    DefaultMessage:Exec
    DefaultCriticalMessage:Exec
    CommIntro:Exec
    CommMessage:Exec
    CommCriticalMessage:Exec
    CornerIntro:Exec
    CornerMessage:Exec
    CornerCriticalMessage:Exec
    RespawnTurningOffWarning:Exec
    RespawnOffWarning:Exec
    CenterMessage:Exec
    CenterImportantMessage:Exec
    CenterCriticalMessage:Exec
   ) {is=title K2Node_SwitchEnum}
   SetTipMessageContent!(execute=@n12.CornerMessage, self, Text:Text=@n11.MsgText) -> (then)
   > Set ContextTutorialOwner(ContextTutorialOwner:Fort Contextual Tutorial Controller Object Reference=@n9.OwnedController) -> (Output_Get:Fort Contextual Tutorial Controller Object Reference)
   > SetVisibility(self=Get Overlay_Container() -> (Overlay_Container:Overlay Object Reference), InVisibility=HitTestInvisible)
   > n13 = Branch(Condition=Get bIsShown() -> (bIsShown:bool))
   n14 = Set Timer by Event(execute=[@n19.then, @n13.then], Delegate=Create Event(), Time=@n11.DisplayTime, bLooping=false)
   > Set OutroAnimationTimerHandle(OutroAnimationTimerHandle:Timer Handle Structure=@n14) -> (Output_Get:Timer Handle Structure)
   > n15 = Is Valid(InputObject=@n11.MsgMediaSource)
   > n16 = OpenSource!(
    execute
    self:Media Player Object Reference=Get MediaPlayer() -> (MediaPlayer:Media Player Object Reference)
    MediaSource:Media Source Object Reference=@n11.MsgMediaSource
   ) -> (then, ReturnValue:bool)
   > n17 = Branch(Condition=@n16)
   n18 = Get Overlay_MediaPlayer() -> (Overlay_MediaPlayer:Overlay Object Reference)
   > SetVisibility(self=@n18, InVisibility=HitTestInvisible)
   SetVisibility(execute=[@n17.else, @n15.Is Not Valid], self=@n18, InVisibility=Hidden)
   StopAllTipAnimations(execute=@n13.else)
   > PlaySound2D(Sound=@n11.MsgSound, VolumeMultiplier=0.25)
   > PlayAnimation(InAnimation=Get Intro())
   > PlaySound2D(Sound=/Game/Sounds/Fort_UI_Sounds/Tutorial/Tutorial_TextPrompt_02_Cue)
   > PlayCloseAnimationIn()
   > n19 = Set bIsShown(bIsShown:bool=true) -> (Output_Get:bool)
   n20 = Event PreConstruct()
   > Branch(Condition=@n20.IsDesignTime)
   > SetText(
    self=Get RichTextBlock_Message() -> (RichTextBlock_Message:Fort Rich Text Block Keybind Object Reference)
    InText="To Text (String)"(InString=Get DevDisplayString() -> (DevDisplayString:String))
   )
   Event Construct()
   > Parent: Construct()
   > SetMobilePosition()
   > n21 = Branch(Condition=Get bAttachToGameState() -> (bAttachToGameState:bool))
   > n22 = Cast To FortGameStateAthena!(execute, Object=GetGameState!() -> (ReturnValue:Game State Base Object Reference)) -> (then, CastFailed, AsFort Game State Athena:Fort Game State Athena Object Reference) {is=title K2Node_DynamicCast}
   > n23 = Bind Event to Context Tutorial Game Mode Message Request(self:Fort Game State Athena Object Reference=@n22, Delegate=Create Event()) {is=title K2Node_AddDelegate}
   SetVisibility(execute=[@n23.then, @n21.else], self=Get Overlay_Container(), InVisibility=Collapsed)
   > n24 = StartListeningToStatefulEvent!(
    execute:Exec
    Target:Gameplay Event Router Component Object Reference=GetGlobalEventRouter()
    EventType:Script Struct Object Reference=/Script/FortniteGame.FortGamePhaseUpdatedEvent
    Context:Object Reference
    EventBubblingRule:EEventBubblingRule Enum
   ) -> (
    then:Exec
    AsyncTaskProxy:Async Action Start Listening To Stateful Event Object Reference
    OnEventReceived:Exec
    OnSavedEventStateExists:Exec
    OnEventStateCleared:Exec
    Payload:Fort Game Phase Updated Event Structure
   ) {is=title K2Node_StartListeningToStatefulEventAsyncAction}
   > Set GamePhaseUpdatedAsyncTask(GamePhaseUpdatedAsyncTask:Async Action Start Listening To Stateful Event Object Reference=@n24) -> (Output_Get:Async Action Start Listening To Stateful Event Object Reference)
   > Bind Event to Hide Game Mode Message Request(self:Fort Game State Athena Object Reference=@n22, Delegate=Create Event()) {is=title K2Node_AddDelegate}
   > Bind Event to Game Mode Message Success(self:Fort Game State Athena Object Reference=@n22, Delegate=Create Event()) {is=title K2Node_AddDelegate}
   > Bind Event to Show Marker Game Mode(self:Fort Game State Athena Object Reference=@n22, Delegate=Create Event()) {is=title K2Node_AddDelegate}
   > Bind Event to Reposition Game Mode Message(self:Fort Game State Athena Object Reference=@n22, Delegate=Create Event()) {is=title K2Node_AddDelegate}
   > Bind Event to Set Default Position Game Mode Message(self:Fort Game State Athena Object Reference=@n22, Delegate=Create Event()) {is=title K2Node_AddDelegate}
   > Bind Event to Hide Marker Game Mode(self:Fort Game State Athena Object Reference=@n22, Delegate=Create Event()) {is=title K2Node_AddDelegate}
   OnGamePhaseChanged!(
    execute=[@n24.OnEventReceived, @n24.OnSavedEventStateExists]
    self
    GamePhase:EAthenaGamePhase Enum=Break Fort Game Phase Updated Event!(FortGamePhaseUpdatedEvent:"Fort Game Phase Updated Event Structure (by ref)"=@n24.Payload) -> (GamePhase:EAthenaGamePhase Enum) {is=title K2Node_BreakStruct}
   ) -> (then) {is=title K2Node_CallFunction}
   n25 = Event OnAnimationFinished!() -> (OutputDelegate, then, Animation:Widget Animation Object Reference) {is=title K2Node_Event}
   > Branch(Condition="Equal (Object)"(A=@n25.Animation, B=Get Outro() -> (Outro:Widget Animation Object Reference)))
   > SetVisibility(self=Get Overlay_Container(), InVisibility=Hidden)
   > Set bIsShown(bIsShown=false)
   > Close(self:Media Player Object Reference=Get MediaPlayer())
   > SetHasActiveContent(bNewActive=false)
   Event OnButtonCloseImplementation() {is=title K2Node_Event}
   > SendEventClosedCallback()
   > PlayOutro(StopAnimation=true)
   Event Destruct()
   > n26 = Get!(execute:Exec) -> (then:Exec, else:Exec, GamePhaseUpdatedAsyncTask:Async Action Start Listening To Stateful Event Object Reference)
   > StopListeningToStatefulEvent(self:Async Action Start Listening To Stateful Event Object Reference=@n26)
   Event StopIntroAnimation() {is=title K2Node_Event}
   > StopAllAnimations(self:User Widget Object Reference)
   Event StartWorldMarkerIntroAnimation() {is=title K2Node_Event}
   > PlayAnimation(InAnimation=Get WorldMarker_intro() -> (WorldMarker_intro:Widget Animation Object Reference))
   > PlayCloseAnimationIn()
   Reroute Node(InputPin:Wildcard) -> (OutputPin:Wildcard)
  ]}
 HUD_ContextualTips_Base_AutoGenFunc {nodes [
   HUD_ContextualTips_Base_AutoGenFunc() {is=title K2Node_FunctionEntry}
   > SendEventClosedCallback()
   > Return Node!(execute, bPassThrough:bool=true)
  ]}
 IsOwner {nodes [
   n1 = IsOwner!() -> (then, TargetOwner:Fort Player Controller Object Reference) {is=title K2Node_FunctionEntry}
   n2 = GetOwningPlayer()
   > Return Node!(execute, IsOwner:bool=OR Boolean(A=IsNullItem(ItemObject=@n1), B="Equal (Object)"(A=@n1, B=@n2)))
  ]}
 On_MobileCloseButtonArea_MouseButtonDown {nodes [
   On_MobileCloseButtonArea_MouseButtonDown!() -> (then, MyGeometry:Geometry Structure, MouseEvent:"Pointer Event Structure (by ref)") {is=title K2Node_FunctionEntry}
   > SendEventClosedCallback()
   > PlayOutroAnimation()
   > n1 = Branch()
   > Return Node!(execute, ReturnValue:Event Reply Structure=Handled())
   Return Node(execute=@n1.else, ReturnValue=Unhandled())
  ]}
 PlayCloseAnimationIn {nodes [
   PlayCloseAnimationIn() {is=title K2Node_FunctionEntry}
   > n1 = Switch on ECommonInputType!(execute, Selection:ECommonInputType Enum=GetCurrentInputType(self=Get CommonInputSubsystem())) -> (MouseAndKeyboard:Exec, Gamepad:Exec, Touch:Exec, Count:Exec) {is=title K2Node_SwitchEnum}
   PlayAnimationForward(execute=[@n1.Gamepad, @n1.MouseAndKeyboard], InAnimation=Get Controller_Tip_In() -> (Controller_Tip_In:Widget Animation Object Reference))
   > n2 = SetCloseText()
   Set HoldTimer(execute=[@n2.then, @n3.then], HoldTimer:double=0) -> (Output_Get:double)
   > SetVisibility(self=Get MobileCLoseButtonArea() -> (MobileCLoseButtonArea:Common Border Object Reference), InVisibility=Visible)
   > ActivateWidget()
   n3 = PlayAnimationForward(execute=@n1.Touch, InAnimation=Get MobileCloseButton_Tip_In() -> (MobileCloseButton_Tip_In:Widget Animation Object Reference))
  ]}
 PlayCloseAnimationOut {nodes [
   PlayCloseAnimationOut() {is=title K2Node_FunctionEntry}
   n1 = Get CommonInputSubsystem()
   > Is Valid(InputObject=@n1)
   > n2 = Switch on ECommonInputType(Selection=GetCurrentInputType(self=@n1))
   n3 = PlayAnimationForward(execute=[@n2.Gamepad, @n2.MouseAndKeyboard], InAnimation=Get Controller_Tip_Out() -> (Controller_Tip_Out:Widget Animation Object Reference))
   SetVisibility(execute=[@n3.then, @n4.then], self=Get MobileCLoseButtonArea(), InVisibility=Hidden)
   > Clear and Invalidate Timer by Handle(Handle=Get HoldButtonTimerHandle() -> (HoldButtonTimerHandle:Timer Handle Structure))
   > DeactivateWidget()
   n4 = PlayAnimationReverse(execute=@n2.Touch, InAnimation=Get MobileCloseButton_Tip_In())
  ]}
 PlayOutro {nodes [
   n1 = PlayOutro!() -> (then, StopAnimation:bool=true) {is=title K2Node_FunctionEntry}
   > Clear and Invalidate Timer by Handle(Handle=Get OutroAnimationTimerHandle())
   > n2 = Branch(Condition=@n1)
   > n3 = StopAllTipAnimations()
   n4 = Get Overlay_Container()
   Branch(
    execute=[@n3.then, @n2.else]
    Condition="Not Equal (Enum)"(
     A:ESlateVisibility Enum=GetVisibility!(self:Widget Object Reference=@n4) -> (ReturnValue:ESlateVisibility Enum=Visible)
     B:ESlateVisibility Enum=Collapsed
    )
   )
   > PlayAnimation(InAnimation=Get Outro())
   > SpawnSound2D!(
    execute
    Sound:Sound Base Object Reference=Get TipCloseSound() -> (TipCloseSound:Sound Base Object Reference)
    VolumeMultiplier:float=1
    PitchMultiplier:float=1
    StartTime:float=0
    ConcurrencySettings:Sound Concurrency Object Reference
    bPersistAcrossLevelTransition:bool=false
    bAutoDestroy:bool=true
   ) -> (then, ReturnValue:Audio Component Object Reference)
   > PlayCloseAnimationOut() {is=title K2Node_CallFunction}
   > Return Node!(execute)
  ]}
 PlayOutroAnimation {nodes [
   PlayOutroAnimation() {is=title K2Node_FunctionEntry}
   > PlayOutro(StopAnimation=true) {is=title K2Node_CallFunction}
  ]}
 SendEventClosedCallback {nodes [
   SendEventClosedCallback() {is=title K2Node_FunctionEntry}
   n1 = Get ContextTutorialOwner() -> (ContextTutorialOwner:Fort Contextual Tutorial Controller Object Reference)
   n2 = IsNullItem(ItemObject=@n1)
   > Branch(Condition=AND Boolean!(A=NOT Boolean(A=@n2), B=IsValid(Object=@n1)) -> (ReturnValue))
   > HideGameModeMessageRequest(self:Fort Contextual Tutorial Controller Object Reference=Get ContextTutorialOwner())
  ]}
 SetCloseText {nodes [
   SetCloseText() {is=title K2Node_FunctionEntry}
   > "SetText (Text)"(
    self=Get ControllerTip_Text() -> (ControllerTip_Text:Common Text Object Reference)
    InText=GetDisplayText!(self:Common Action Widget Object Reference=Get TipActionWidget() -> (TipActionWidget:Common Action Widget Object Reference)) -> (ReturnValue:Text)
   )
  ]}
 SetMobilePosition {nodes [
   SetMobilePosition() {is=title K2Node_FunctionEntry}
   n1 = GetOwningPlayer()
   > Branch(Condition=AND Boolean(A=IsMobileGame(), B=IsUsingTouch(PlayerController=@n1)))
   n2 = GetViewportSize!() -> (ReturnValue_X:double, ReturnValue_Y:double)
   > Branch(Condition="float >= float"(A=float / float(A=@n2, B=@n2.ReturnValue_Y), B=Get MobileViewportRatio() -> (MobileViewportRatio:double)))
   > SetPadding!(
    execute
    self:Overlay Slot Object Reference=SlotAsOverlaySlot(Widget=Get CommonBorder_Mobile() -> (CommonBorder_Mobile:Common Border Object Reference))
    InPadding_Left:float=0
    InPadding_Top:float=0
    InPadding_Right:float=0
    InPadding_Bottom:float=0
   ) -> (then)
  ]}
 SetTipMessageContent {nodes [
   n1 = SetTipMessageContent!() -> (then, Text:Text) {is=title K2Node_FunctionEntry}
   > SetText(self=Get RichTextBlock_Message(), InText=@n1)
   > Return Node()
  ]}
 StopAllTipAnimations {nodes [
   StopAllTipAnimations() {is=title K2Node_FunctionEntry}
   > StopAnimation(InAnimation=Get Intro())
   > StopAnimation(InAnimation=Get Feedback_Positive_V2())
   > StopAnimation(InAnimation=Get Outro())
   > Return Node()
  ]}
 UpdateHoldCloseKey {nodes [
   UpdateHoldCloseKey() {is=title K2Node_FunctionEntry}
   > n1 = Branch(Condition=IsVisible(self=Get Overlay_Container()))
   > Return Node!(execute, Handled:bool=true)
   Return Node(execute=@n1.else, Handled=false)
  ]}
]
variables [
 bAttachToGameState {type=boolean, default=true}
 bIsShown {type=boolean}
 ContextTutorialOwner {type=FortContextualTutorialController, default=null}
 CurrentGamePhase {type=EAthenaGamePhase, default=None}
 CurrentlyShownMessage {
  type=AthenaGameMessageData
  default {bIsTeamBased=false, displayTime=5, messageFlags=None, msgDelay=0, msgMediaSource=null, msgSound=null, msgText="", msgType=None, teamIndex=0}
 }
 DevDisplayString {type=string, default="Lorem <ENEMY>Enemy</> or <Ally>Ally</> Dolor amit <b>sequitor</> <KEYWORD>keyword</>!"}
 GamePhaseUpdatedAsyncTask {type=AsyncAction_StartListeningToStatefulEvent, default=null}
 HoldButtonTimerHandle {type=TimerHandle}
 HoldTimer {type=number}
 MediaPlayer {type=MediaPlayer, default=/Game/Athena/HUD/LTM/Tutorial/Tut_MediaPlayer}
 MobileViewportRatio {type=number, default=1.34}
 OutroAnimationTimerHandle {type=TimerHandle}
 TargetTimer {type=number}
 TipCloseSound {type=SoundBase, default=/Game/Sounds/Fort_UI_Sounds/Gameplay/UI_TipWindow_Close_Cue}
]
widgets {
 info {
  inheritedWidgetCount=7
  namedSlotCount=1
  parentClass=/Script/FortniteUI.FortHudContextTutorialTipsBase
  rootWidgetClass=/Game/Athena/HUD/FortHUDElementWrapper.FortHUDElementWrapper_C
  widgetCount=60
 }
 widgets [
  FortHUDElementWrapper {
   widgetClassPath=/Game/Athena/HUD/FortHUDElementWrapper
   bIsVariable=true
   widget=FortHUDElementWrapper
   properties {elementTag {tagName=HUD.Athena.ContextualTooltip}}
  }
  ScreenOverlay {widgetClassPath=&Overlay, namedSlotHost=FortHUDElementWrapper, widget=ScreenOverlay}
  CanvasPanel_0 {
   widgetClassPath=&CanvasPanel
   parent=ScreenOverlay
   widget=CanvasPanel_0
   slot {class=OverlaySlot, properties {horizontalAlignment=HAlign_Center, verticalAlignment=VAlign_Bottom, padding=<ltrb 0 0 0 300>}}
  }
  CanvasPanel_Ball_Parented {
   widgetClassPath=&CanvasPanel
   bIsVariable=true
   parent=CanvasPanel_0
   widget=CanvasPanel_Ball_Parented
   slot {class=CanvasPanelSlot, properties=~}
  }
  ContextTutorialBallParented {
   widgetClassPath=/Game/Athena/Tutorial/Contextual/ContextTutorialBall
   bInherited=true
   bIsVariable=true
   parent=CanvasPanel_Ball_Parented
   widget=ContextTutorialBallParented
   properties {bCanArrowAnimationPlay=true, visibility=Hidden, elementTag {tagName=Athena.HideForCapture}}
   slot {
    class=CanvasPanelSlot
    properties {zOrder=1, layoutData {alignment=<0 0>, anchors {maximum=<0 0>, minimum=<0 0>}, offsets=<ltrb 0 0 100 100>}}
   }
  }
  CommonBorder_Mobile {
   widgetClassPath=&CommonBorder
   bIsVariable=true
   parent=ScreenOverlay
   widget=CommonBorder_Mobile
   properties {horizontalAlignment=HAlign_Left, verticalAlignment=VAlign_Center, padding=<ltrb 0 0 0 0>}
   slot {class=OverlaySlot, properties {verticalAlignment=VAlign_Center, padding=<ltrb 0 0 0 100>}}
  }
  CanvasPanel_Container {
   widgetClassPath=&CanvasPanel
   bInherited=true
   bIsVariable=true
   parent=CommonBorder_Mobile
   widget=CanvasPanel_Container
   slot {class=BorderSlot, properties {horizontalAlignment=HAlign_Left, verticalAlignment=VAlign_Center, padding=<ltrb 0 0 0 0>}}
  }
  CanvasPanel_Ball {
   widgetClassPath=&CanvasPanel
   bInherited=true
   bIsVariable=true
   parent=CanvasPanel_Container
   widget=CanvasPanel_Ball
   slot {
    class=CanvasPanelSlot
    properties {zOrder=1, layoutData {alignment=<0 0>, anchors {maximum=<0.5 0.5>, minimum=<0.5 0.5>}, offsets=<ltrb 137.1564 -119.8024 100 100>}}
   }
  }
  ContextTutorialBall {
   widgetClassPath=&ContextTutorialBall
   bInherited=true
   bIsVariable=true
   parent=CanvasPanel_Ball
   widget=ContextTutorialBall
   properties {visibility=Hidden, elementTag {tagName=Athena.HideForCapture}}
   slot {
    class=CanvasPanelSlot
    properties {zOrder=1, layoutData {alignment=<0 0>, anchors {maximum=<0 0>, minimum=<0 0>}, offsets=<ltrb 0 0 100 100>}}
   }
  }
  Overlay_Container {
   widgetClassPath=&Overlay
   bInherited=true
   bIsVariable=true
   parent=CanvasPanel_Container
   widget=Overlay_Container
   properties {visibility=HitTestInvisible, renderTransform {angle=0, scale=<1 1>, shear=<0 0>, translation=<40 0>}}
   slot {
    class=CanvasPanelSlot
    properties {bAutoSize=true, layoutData {alignment=<0 0>, anchors {maximum=<0 0>, minimum=<0 0>}, offsets=<ltrb 0 0 550 219>}}
   }
  }
  Overlay_5 {
   widgetClassPath=&Overlay
   parent=Overlay_Container
   widget=Overlay_5
   slot {class=OverlaySlot, properties {horizontalAlignment=HAlign_Fill, verticalAlignment=VAlign_Fill}}
  }
  Overlay_PositiveFeedback {
   widgetClassPath=&Overlay
   parent=Overlay_5
   widget=Overlay_PositiveFeedback
   properties {bIsEnabled=false, renderOpacity=0, visibility=Collapsed, renderTransform {angle=0, scale=<1 1>, shear=<-10 0>, translation=<0 0>}}
   slot {class=OverlaySlot, properties {horizontalAlignment=HAlign_Fill, verticalAlignment=VAlign_Fill}}
  }
  PositiveFeedback_BG {
   widgetClassPath=&Image
   parent=Overlay_PositiveFeedback
   widget=PositiveFeedback_BG
   properties {
    brush {
     bIsDynamicallyLoaded=false
     drawAs=Box
     imageType=NoImage
     mirroring=NoMirror
     resourceName=None
     resourceObject=/Game/Athena/HUD/LTM/Tutorial/Materials/M_UI_SkewBoxBorder_Inst
     tiling=NoTile
     imageSize=<64 64>
     margin=<ltrb 0.2 0.2 0.2 0.2>
     outlineSettings {
      bUseBrushTransparency=false
      roundingType=HalfHeightRadius
      width=0
      color {colorUseRule=UseColor_Specified, specifiedColor=<rgba 0 0 0 0>}
      cornerRadii=<0 0 0 0>
     }
     tintColor {colorUseRule=UseColor_Specified, specifiedColor=<rgba 0.0343 1 0.227 1>}
     uVRegion {bIsValid=false, max=<0 0>, min=<0 0>}
    }
    renderTransform {angle=0, scale=<1 1>, shear=<-5.0682 0>, translation=<0 0>}
   }
   slot {class=OverlaySlot, properties {horizontalAlignment=HAlign_Fill, verticalAlignment=VAlign_Fill}}
  }
  Overlay_PositiveFeedback_2 {
   widgetClassPath=&Overlay
   parent=Overlay_5
   widget=Overlay_PositiveFeedback_2
   properties {bIsEnabled=false, renderOpacity=0, visibility=Hidden, renderTransform {angle=0, scale=<1 1>, shear=<-10 0>, translation=<0 0>}}
   slot {class=OverlaySlot, properties {horizontalAlignment=HAlign_Fill, verticalAlignment=VAlign_Fill}}
  }
  PositiveFeedback_BG_1 {
   widgetClassPath=&Image
   parent=Overlay_PositiveFeedback_2
   widget=PositiveFeedback_BG_1
   properties {
    brush {
     bIsDynamicallyLoaded=false
     drawAs=Box
     imageType=NoImage
     mirroring=NoMirror
     resourceName=None
     resourceObject=&M_UI_SkewBoxBorder_Inst
     tiling=NoTile
     imageSize=<64 64>
     margin=<ltrb 0.2 0.2 0.2 0.2>
     outlineSettings {
      bUseBrushTransparency=false
      roundingType=HalfHeightRadius
      width=0
      color {colorUseRule=UseColor_Specified, specifiedColor=<rgba 0 0 0 0>}
      cornerRadii=<0 0 0 0>
     }
     tintColor {colorUseRule=UseColor_Specified, specifiedColor=<rgba 0.0343 1 0.227 1>}
     uVRegion {bIsValid=false, max=<0 0>, min=<0 0>}
    }
    renderTransform {angle=0, scale=<1 1>, shear=<-5.0682 0>, translation=<0 0>}
   }
   slot {class=OverlaySlot, properties {horizontalAlignment=HAlign_Fill, verticalAlignment=VAlign_Fill}}
  }
  Overlay_Dialog {
   widgetClassPath=&Overlay
   bIsVariable=true
   parent=Overlay_Container
   widget=Overlay_Dialog
   properties {renderTransformPivot=<0 0.6439>}
   slot {class=OverlaySlot, properties {verticalAlignment=VAlign_Center}}
  }
  OverlayTextContent {
   widgetClassPath=&Overlay
   parent=Overlay_Dialog
   widget=OverlayTextContent
   slot {class=OverlaySlot, properties {horizontalAlignment=HAlign_Fill, verticalAlignment=VAlign_Center}}
  }
  Animated_Close {
   widgetClassPath=&Overlay
   parent=OverlayTextContent
   widget=Animated_Close
   slot {class=OverlaySlot, properties {horizontalAlignment=HAlign_Fill, verticalAlignment=VAlign_Fill}}
  }
  TipForController {
   widgetClassPath=&Overlay
   parent=Animated_Close
   widget=TipForController
   properties {renderOpacity=0, renderTransform {angle=0, scale=<1 1>, shear=<0 0>, translation=<-20 85>}}
   slot {class=OverlaySlot, properties {horizontalAlignment=HAlign_Fill, verticalAlignment=VAlign_Bottom}}
  }
  HorizontalBox_2 {
   widgetClassPath=&HorizontalBox
   parent=TipForController
   widget=HorizontalBox_2
   slot {class=OverlaySlot, properties {horizontalAlignment=HAlign_Fill, padding=<ltrb 40 0 30 0>}}
  }
  TipActionWidget {
   widgetClassPath=&CommonActionWidget
   bInherited=true
   bIsVariable=true
   parent=HorizontalBox_2
   widget=TipActionWidget
   properties {
    progressMaterialParam=Percentage
    inputActions [{dataTable=/Game/Athena/Tutorial/Contextual/ContextTutorial_InputActionDataTable, rowName=CloseTip}]
    progressMaterialBrush {
     bIsDynamicallyLoaded=false
     drawAs=Image
     imageType=NoImage
     mirroring=NoMirror
     resourceName=None
     resourceObject=/Game/Athena/HUD/M_HealProgress
     tiling=NoTile
     imageSize=<80 80>
     margin=<ltrb 0 0 0 0>
     outlineSettings {
      bUseBrushTransparency=false
      roundingType=HalfHeightRadius
      width=0
      color {colorUseRule=UseColor_Specified, specifiedColor=<rgba 0 0 0 0>}
      cornerRadii=<0 0 0 0>
     }
     tintColor {colorUseRule=UseColor_Specified, specifiedColor=<rgba 1 1 1 1>}
     uVRegion {bIsValid=false, max=<0 0>, min=<0 0>}
    }
    renderTransform {angle=0, scale=<1.3 1.3>, shear=<0 0>, translation=<0 0>}
   }
   slot {class=HorizontalBoxSlot, properties {padding=<ltrb 0 0 15 0>}}
  }
  ControllerTip_Text {
   widgetClassPath=&CommonTextBlock
   bIsVariable=true
   parent=HorizontalBox_2
   widget=ControllerTip_Text
   properties {
    scrollStyle=/Game/Athena/Prototype/Blueprints/Galileo/UI/TextScrollStyle-Base-Ticker.TextScrollStyle-Base-Ticker_C
    text="(HOLD) Close Tip"
    font {
     bForceMonospaced=false
     bMaterialIsStencil=false
     fontMaterial=null
     fontObject=&BurbankSmall
     letterSpacing=0
     monospacedWidth=1
     size=20
     skewAmount=0
     typefaceFontName=Bold
     outlineSettings {
      bApplyOutlineToDropShadows=false
      bMiteredCorners=false
      bSeparateFillAlpha=false
      outlineMaterial=null
      outlineSize=0
      outlineColor=<rgba 0 0 0 1>
     }
    }
    shadowOffset=<0 0>
   }
   slot {class=HorizontalBoxSlot, properties {verticalAlignment=VAlign_Center, size {sizeRule=Fill, value=1}}}
  }
  Overlay_MediaPlayer {
   widgetClassPath=&Overlay
   bIsVariable=true
   parent=Animated_Close
   widget=Overlay_MediaPlayer
   properties {renderTransform {angle=0, scale=<1 1>, shear=<0 0>, translation=<0 290>}}
   slot {class=OverlaySlot, properties {horizontalAlignment=HAlign_Center}}
  }
  Image_Media {
   widgetClassPath=&Image
   bIsVariable=true
   parent=Overlay_MediaPlayer
   widget=Image_Media
   properties {
    visibility=HitTestInvisible
    brush {
     bIsDynamicallyLoaded=false
     drawAs=Image
     imageType=NoImage
     mirroring=NoMirror
     resourceName=None
     resourceObject=/Game/Athena/HUD/LTM/Tutorial/Tut_MediaPlayer_Video_Mat
     tiling=NoTile
     imageSize=<326.1676 166.0381>
     margin=<ltrb 0 0 0 0>
     outlineSettings {
      bUseBrushTransparency=false
      roundingType=HalfHeightRadius
      width=0
      color {colorUseRule=UseColor_Specified, specifiedColor=<rgba 0 0 0 0>}
      cornerRadii=<0 0 0 0>
     }
     tintColor {colorUseRule=UseColor_Specified, specifiedColor=<rgba 1 1 1 1>}
     uVRegion {bIsValid=false, max=<0 0>, min=<0 0>}
    }
   }
   slot {class=OverlaySlot, properties {horizontalAlignment=HAlign_Fill, verticalAlignment=VAlign_Fill}}
  }
  OverlayBG {
   widgetClassPath=&Overlay
   parent=OverlayTextContent
   widget=OverlayBG
   slot {class=OverlaySlot, properties {horizontalAlignment=HAlign_Fill, verticalAlignment=VAlign_Fill}}
  }
  Image_BGDropShadow {
   widgetClassPath=&Image
   parent=OverlayBG
   widget=Image_BGDropShadow
   properties {
    visibility=Collapsed
    brush {
     bIsDynamicallyLoaded=false
     drawAs=Box
     imageType=NoImage
     mirroring=NoMirror
     resourceName=None
     resourceObject=/Game/Athena/HUD/LTM/Tutorial/Materials/M_UI_SkewBox_TutorialBG1
     tiling=NoTile
     imageSize=<256 128>
     margin=<ltrb 0.2 0 0.2 0>
     outlineSettings {
      bUseBrushTransparency=false
      roundingType=HalfHeightRadius
      width=0
      color {colorUseRule=UseColor_Specified, specifiedColor=<rgba 0 0 0 0>}
      cornerRadii=<0 0 0 0>
     }
     tintColor {colorUseRule=UseColor_Specified, specifiedColor=<rgba 0.0052 0.0052 0.0056 0.4>}
     uVRegion {bIsValid=false, max=<0 0>, min=<0 0>}
    }
    renderTransform {angle=1.5, scale=<1 1>, shear=<-2 0>, translation=<13 13>}
    renderTransformPivot=<0.2 0.5>
   }
   slot {class=OverlaySlot, properties {horizontalAlignment=HAlign_Fill, verticalAlignment=VAlign_Fill}}
  }
  MobileCloseBUtton {
   widgetClassPath=&Overlay
   parent=OverlayBG
   widget=MobileCloseBUtton
   properties {renderOpacity=0, visibility=Visible, renderTransform {angle=0, scale=<1 1>, shear=<0 0>, translation=<61.9627 2>}}
   slot {class=OverlaySlot, properties {horizontalAlignment=HAlign_Right}}
  }
  SizeBox_2 {
   widgetClassPath=&SizeBox
   parent=MobileCloseBUtton
   widget=SizeBox_2
   properties {bOverride_HeightOverride=true, bOverride_WidthOverride=true, heightOverride=80, visibility=Visible, widthOverride=80}
   slot {class=OverlaySlot}
  }
  CloseButtonBaking {widgetClassPath=&Overlay, parent=SizeBox_2, widget=CloseButtonBaking, slot {class=SizeBoxSlot}}
  Image_BGCloseMobileButton {
   widgetClassPath=&Image
   parent=CloseButtonBaking
   widget=Image_BGCloseMobileButton
   properties {
    brush {
     bIsDynamicallyLoaded=false
     drawAs=Box
     imageType=NoImage
     mirroring=NoMirror
     resourceName=None
     resourceObject=&M_UI_SkewBox_TutorialBG1
     tiling=NoTile
     imageSize=<45 90>
     margin=<ltrb 0.2 0 0.2 0>
     outlineSettings {
      bUseBrushTransparency=false
      roundingType=HalfHeightRadius
      width=0
      color {colorUseRule=UseColor_Specified, specifiedColor=<rgba 0 0 0 0>}
      cornerRadii=<0 0 0 0>
     }
     tintColor {colorUseRule=UseColor_Specified, specifiedColor=<rgba 0 0.117 1 1>}
     uVRegion {bIsValid=false, max=<0 0>, min=<0 0>}
    }
    renderTransform {angle=1.5, scale=<1 1>, shear=<-2 0>, translation=<0 0>}
    renderTransformPivot=<0.2 0.5>
   }
   slot {class=OverlaySlot, properties {horizontalAlignment=HAlign_Fill, verticalAlignment=VAlign_Fill}}
  }
  Image_IconMobileButton {
   widgetClassPath=&Image
   parent=CloseButtonBaking
   widget=Image_IconMobileButton
   properties {
    brush {
     bIsDynamicallyLoaded=false
     drawAs=Image
     imageType=NoImage
     mirroring=NoMirror
     resourceName=None
     resourceObject=/Game/UI/Foundation/Textures/Icons/Checks/T-Icon-X-128
     tiling=NoTile
     imageSize=<44.3956 41.2982>
     margin=<ltrb 0.2 0 0.2 0>
     outlineSettings {
      bUseBrushTransparency=false
      roundingType=HalfHeightRadius
      width=0
      color {colorUseRule=UseColor_Specified, specifiedColor=<rgba 0 0 0 0>}
      cornerRadii=<0 0 0 0>
     }
     tintColor {colorUseRule=UseColor_Specified, specifiedColor=<rgba 1 1 1 1>}
     uVRegion {bIsValid=false, max=<0 0>, min=<0 0>}
    }
    renderTransform {angle=1.5, scale=<1 1>, shear=<-2 0>, translation=<0 0>}
    renderTransformPivot=<0.2 0.5>
   }
   slot {class=OverlaySlot, properties {horizontalAlignment=HAlign_Center, verticalAlignment=VAlign_Center}}
  }
  Image_BGCloseMobileButton_Shadow {
   widgetClassPath=&Image
   parent=CloseButtonBaking
   widget=Image_BGCloseMobileButton_Shadow
   properties {
    brush {
     bIsDynamicallyLoaded=false
     drawAs=Image
     imageType=NoImage
     mirroring=NoMirror
     resourceName=None
     resourceObject=&M_UI_SkewBox_TutorialBG1
     tiling=NoTile
     imageSize=<31 90>
     margin=<ltrb 0 0 0 0>
     outlineSettings {
      bUseBrushTransparency=false
      roundingType=HalfHeightRadius
      width=0
      color {colorUseRule=UseColor_Specified, specifiedColor=<rgba 0 0 0 0>}
      cornerRadii=<0 0 0 0>
     }
     tintColor {colorUseRule=UseColor_Specified, specifiedColor=<rgba 0 0 0 0.17>}
     uVRegion {bIsValid=false, max=<0 0>, min=<0 0>}
    }
    renderTransform {angle=1.5, scale=<1 1>, shear=<-5 0>, translation=<0 0>}
    renderTransformPivot=<0.2 0.5>
   }
   slot {class=OverlaySlot, properties {verticalAlignment=VAlign_Fill}}
  }
  Image_BGOutline {
   widgetClassPath=&Image
   parent=OverlayBG
   widget=Image_BGOutline
   properties {
    visibility=SelfHitTestInvisible
    brush {
     bIsDynamicallyLoaded=false
     drawAs=Box
     imageType=NoImage
     mirroring=NoMirror
     resourceName=None
     resourceObject=&M_UI_SkewBox_TutorialBG1
     tiling=NoTile
     imageSize=<256 128>
     margin=<ltrb 0.2 0 0.2 0>
     outlineSettings {
      bUseBrushTransparency=false
      roundingType=HalfHeightRadius
      width=0
      color {colorUseRule=UseColor_Specified, specifiedColor=<rgba 0 0 0 0>}
      cornerRadii=<0 0 0 0>
     }
     tintColor {colorUseRule=UseColor_Specified, specifiedColor=<rgba 1 1 1 1>}
     uVRegion {bIsValid=false, max=<0 0>, min=<0 0>}
    }
    renderTransform {angle=0, scale=<1 1>, shear=<-1 0>, translation=<0 0>}
    renderTransformPivot=<0 0.5>
   }
   slot {class=OverlaySlot, properties {horizontalAlignment=HAlign_Fill, verticalAlignment=VAlign_Fill}}
  }
  Image_BGInset {
   widgetClassPath=&Image
   parent=OverlayBG
   widget=Image_BGInset
   properties {
    visibility=SelfHitTestInvisible
    brush {
     bIsDynamicallyLoaded=false
     drawAs=Box
     imageType=NoImage
     mirroring=NoMirror
     resourceName=None
     resourceObject=&M_UI_SkewBox_TutorialBG1
     tiling=NoTile
     imageSize=<256 128>
     margin=<ltrb 0.2 0 0.3 0>
     outlineSettings {
      bUseBrushTransparency=false
      roundingType=HalfHeightRadius
      width=0
      color {colorUseRule=UseColor_Specified, specifiedColor=<rgba 0 0 0 0>}
      cornerRadii=<0 0 0 0>
     }
     tintColor {colorUseRule=UseColor_Specified, specifiedColor=<rgba 1 1 1 1>}
     uVRegion {bIsValid=false, max=<0 0>, min=<0 0>}
    }
    colorAndOpacity=<rgba 0.0273 0.0273 0.0273 1>
    renderTransform {angle=0, scale=<1 1>, shear=<0 0>, translation=<0 -1>}
   }
   slot {class=OverlaySlot, properties {horizontalAlignment=HAlign_Fill, verticalAlignment=VAlign_Fill, padding=<ltrb 5 5 5 5>}}
  }
  Image_BGInset_Corner {
   widgetClassPath=&Image
   parent=OverlayBG
   widget=Image_BGInset_Corner
   properties {
    visibility=SelfHitTestInvisible
    brush {
     bIsDynamicallyLoaded=false
     drawAs=Image
     imageType=NoImage
     mirroring=NoMirror
     resourceName=None
     resourceObject=/Game/Athena/HUD/LTM/Tutorial/Materials/M_UI_SkewBox_TutorialCornerBG
     tiling=NoTile
     imageSize=<0 0>
     margin=<ltrb 0 0 1 0>
     outlineSettings {
      bUseBrushTransparency=false
      roundingType=HalfHeightRadius
      width=0
      color {colorUseRule=UseColor_Specified, specifiedColor=<rgba 0 0 0 0>}
      cornerRadii=<0 0 0 0>
     }
     tintColor {colorUseRule=UseColor_Specified, specifiedColor=<rgba 1 1 1 1>}
     uVRegion {bIsValid=false, max=<0 0>, min=<0 0>}
    }
    colorAndOpacity=<rgba 0 0.117 1 1>
   }
   slot {class=OverlaySlot, properties {horizontalAlignment=HAlign_Fill, verticalAlignment=VAlign_Fill, padding=<ltrb 0 5 0 0>}}
  }
  Image_BGInset_Corner_Solo {
   widgetClassPath=&Image
   parent=OverlayTextContent
   widget=Image_BGInset_Corner_Solo
   properties {
    visibility=Collapsed
    brush {
     bIsDynamicallyLoaded=false
     drawAs=Image
     imageType=NoImage
     mirroring=NoMirror
     resourceName=None
     resourceObject=&M_UI_SkewBox_TutorialCornerBG
     tiling=NoTile
     imageSize=<0 0>
     margin=<ltrb 0 0 1 0>
     outlineSettings {
      bUseBrushTransparency=false
      roundingType=HalfHeightRadius
      width=0
      color {colorUseRule=UseColor_Specified, specifiedColor=<rgba 0 0 0 0>}
      cornerRadii=<0 0 0 0>
     }
     tintColor {colorUseRule=UseColor_Specified, specifiedColor=<rgba 1 1 1 1>}
     uVRegion {bIsValid=false, max=<0 0>, min=<0 0>}
    }
    colorAndOpacity=<rgba 0 0.117 1 1>
   }
   slot {class=OverlaySlot, properties {horizontalAlignment=HAlign_Fill, verticalAlignment=VAlign_Fill, padding=<ltrb 0 5 0 0>}}
  }
  BorderText {
   widgetClassPath=&Border
   parent=OverlayTextContent
   widget=BorderText
   properties {
    visibility=HitTestInvisible
    background {
     bIsDynamicallyLoaded=false
     drawAs=NoDrawType
     imageType=NoImage
     mirroring=NoMirror
     resourceName=None
     resourceObject=null
     tiling=NoTile
     imageSize=<32 32>
     margin=<ltrb 0 0 0 0>
     outlineSettings {
      bUseBrushTransparency=false
      roundingType=HalfHeightRadius
      width=0
      color {colorUseRule=UseColor_Specified, specifiedColor=<rgba 0 0 0 0>}
      cornerRadii=<0 0 0 0>
     }
     tintColor {colorUseRule=UseColor_Specified, specifiedColor=<rgba 1 1 1 1>}
     uVRegion {bIsValid=false, max=<0 0>, min=<0 0>}
    }
    padding=<ltrb 0 0 0 0>
   }
   slot {class=OverlaySlot, properties {horizontalAlignment=HAlign_Fill, verticalAlignment=VAlign_Fill}}
  }
  SizeBox_Message {
   widgetClassPath=&SizeBox
   parent=BorderText
   widget=SizeBox_Message
   properties {maxDesiredWidth=360, minDesiredWidth=340, renderTransformPivot=<0.5 1>}
   slot {class=BorderSlot, properties {padding=<ltrb 0 0 0 0>}}
  }
  RichTextBlock_Message {
   widgetClassPath=/Script/FortniteUI.FortRichTextBlockKeybind
   bInherited=true
   bIsVariable=true
   parent=SizeBox_Message
   widget=RichTextBlock_Message
   properties {
    autoWrapText=true
    keybindWidgetClass=/FortUILibrary/Input/KeybindWidget.KeybindWidget_C
    text="LOREM <ENEMY>ENEMY</> OR <ALLY>ALLY</> DOLOR AMIT <B>SEQUITOR</> <KEYWORD>KEYWORD</>!"
    textStyleSet=/Game/Athena/HUD/LTM/Tutorial/Materials/ContextualTutorialCalloutStyles
    textTransformPolicy=ToUpper
    wrapTextAt=385
    decoratorClasses [/Game/Athena/HUD/TextStyles/ContextualTutorialImageDecorator.ContextualTutorialImageDecorator_C]
    defaultTextStyleOverride {
     overflowPolicy=Clip
     transformPolicy=None
     colorAndOpacity {colorUseRule=UseColor_Specified, specifiedColor=<rgba 1 1 1 1>}
     font {
      bForceMonospaced=false
      bMaterialIsStencil=false
      fontMaterial=null
      fontObject=&NotoSans
      letterSpacing=0
      monospacedWidth=1
      size=24
      skewAmount=0
      typefaceFontName=NotoSans-Regular
      outlineSettings {
       bApplyOutlineToDropShadows=false
       bMiteredCorners=false
       bSeparateFillAlpha=false
       outlineMaterial=null
       outlineSize=0
       outlineColor=<rgba 0 0 0 1>
      }
     }
     highlightColor {colorUseRule=UseColor_Specified, specifiedColor=<rgba 0 0 0 1>}
     highlightShape {
      bIsDynamicallyLoaded=false
      drawAs=Image
      imageType=NoImage
      mirroring=NoMirror
      resourceName=None
      resourceObject=null
      tiling=NoTile
      imageSize=<32 32>
      margin=<ltrb 0 0 0 0>
      outlineSettings {
       bUseBrushTransparency=false
       roundingType=HalfHeightRadius
       width=0
       color {colorUseRule=UseColor_Specified, specifiedColor=<rgba 0 0 0 0>}
       cornerRadii=<0 0 0 0>
      }
      tintColor {colorUseRule=UseColor_Specified, specifiedColor=<rgba 1 1 1 1>}
      uVRegion {bIsValid=false, max=<0 0>, min=<0 0>}
     }
     selectedBackgroundColor {colorUseRule=UseColor_Foreground, specifiedColor=<rgba 1 0 1 1>}
     shadowColorAndOpacity=<rgba 0 0 0 1>
     shadowOffset=<0 0>
     strikeBrush {
      bIsDynamicallyLoaded=false
      drawAs=Image
      imageType=NoImage
      mirroring=NoMirror
      resourceName=None
      resourceObject=null
      tiling=NoTile
      imageSize=<32 32>
      margin=<ltrb 0 0 0 0>
      outlineSettings {
       bUseBrushTransparency=false
       roundingType=HalfHeightRadius
       width=0
       color {colorUseRule=UseColor_Specified, specifiedColor=<rgba 0 0 0 0>}
       cornerRadii=<0 0 0 0>
      }
      tintColor {colorUseRule=UseColor_Specified, specifiedColor=<rgba 1 1 1 1>}
      uVRegion {bIsValid=false, max=<0 0>, min=<0 0>}
     }
     underlineBrush {
      bIsDynamicallyLoaded=false
      drawAs=Image
      imageType=NoImage
      mirroring=NoMirror
      resourceName=None
      resourceObject=null
      tiling=NoTile
      imageSize=<32 32>
      margin=<ltrb 0 0 0 0>
      outlineSettings {
       bUseBrushTransparency=false
       roundingType=HalfHeightRadius
       width=0
       color {colorUseRule=UseColor_Specified, specifiedColor=<rgba 0 0 0 0>}
       cornerRadii=<0 0 0 0>
      }
      tintColor {colorUseRule=UseColor_Specified, specifiedColor=<rgba 1 1 1 1>}
      uVRegion {bIsValid=false, max=<0 0>, min=<0 0>}
     }
    }
    renderTransform {angle=0, scale=<1 1>, shear=<-8 0>, translation=<0 0>}
   }
   slot {class=SizeBoxSlot, properties {verticalAlignment=VAlign_Center, padding=<ltrb 40 25 50 20>}}
  }
  Overlay_2 {
   widgetClassPath=&Overlay
   parent=OverlayTextContent
   widget=Overlay_2
   properties {renderTransform {angle=0, scale=<1.1 0.95>, shear=<0 0>, translation=<0 0>}, renderTransformPivot=<0.5 0>}
   slot {class=OverlaySlot, properties {horizontalAlignment=HAlign_Right}}
  }
  CommonTextBlock_40 {
   widgetClassPath=&CommonTextBlock
   parent=Overlay_2
   widget=CommonTextBlock_40
   properties {
    text="!"
    visibility=HitTestInvisible
    font {
     bForceMonospaced=false
     bMaterialIsStencil=false
     fontMaterial=null
     fontObject=&Burbank
     letterSpacing=0
     monospacedWidth=1
     size=40
     skewAmount=0
     typefaceFontName=Burbank-Black
     outlineSettings {
      bApplyOutlineToDropShadows=false
      bMiteredCorners=false
      bSeparateFillAlpha=false
      outlineMaterial=null
      outlineSize=0
      outlineColor=<rgba 0 0 0 1>
     }
    }
    renderTransform {angle=0, scale=<1.25 1.25>, shear=<-12 0>, translation=<-1 -8>}
    shadowColorAndOpacity=<rgba 0.0052 0.0052 0.0056 0.408>
    shadowOffset=<4 4>
   }
   slot {class=OverlaySlot, properties {horizontalAlignment=HAlign_Center, verticalAlignment=VAlign_Center}}
  }
  OverlayCheck {
   widgetClassPath=&Overlay
   parent=Overlay_2
   widget=OverlayCheck
   properties {visibility=HitTestInvisible, renderTransform {angle=0, scale=<1.6 1.6>, shear=<0 0>, translation=<-4 -32>}}
   slot {class=OverlaySlot, properties {horizontalAlignment=HAlign_Center, verticalAlignment=VAlign_Center}}
  }
  Overlaytriangles {
   widgetClassPath=&Overlay
   parent=OverlayCheck
   widget=Overlaytriangles
   properties {renderOpacity=0}
   slot {class=OverlaySlot, properties {horizontalAlignment=HAlign_Center, verticalAlignment=VAlign_Fill, padding=<ltrb -50 0 -50 0>}}
  }
  HorizontalBox {
   widgetClassPath=&HorizontalBox
   parent=Overlaytriangles
   widget=HorizontalBox
   properties {renderTransform {angle=0, scale=<1 1>, shear=<-12 0>, translation=<0 0>}}
   slot {class=OverlaySlot, properties {horizontalAlignment=HAlign_Fill, verticalAlignment=VAlign_Center}}
  }
  ImageTriangleL_1 {
   widgetClassPath=&Image
   parent=HorizontalBox
   widget=ImageTriangleL_1
   properties {
    brush {
     bIsDynamicallyLoaded=false
     drawAs=Image
     imageType=NoImage
     mirroring=NoMirror
     resourceName=None
     resourceObject=/Game/UI/HUD/Materials/M_FX_TriangleScroll-Burst
     tiling=NoTile
     imageSize=<200 40>
     margin=<ltrb 0 0 0 0>
     outlineSettings {
      bUseBrushTransparency=false
      roundingType=HalfHeightRadius
      width=0
      color {colorUseRule=UseColor_Specified, specifiedColor=<rgba 0 0 0 0>}
      cornerRadii=<0 0 0 0>
     }
     tintColor {colorUseRule=UseColor_Specified, specifiedColor=<rgba 0 1 0.1907 1>}
     uVRegion {bIsValid=false, max=<0 0>, min=<0 0>}
    }
    renderTransform {angle=0, scale=<-1 1>, shear=<0 0>, translation=<0 0>}
    renderTransformPivot=<0.5 1>
   }
   slot {class=HorizontalBoxSlot, properties {horizontalAlignment=HAlign_Right, verticalAlignment=VAlign_Center, padding=<ltrb -200 0 0 0>}}
  }
  Spacer {widgetClassPath=&Spacer, parent=HorizontalBox, widget=Spacer, slot {class=HorizontalBoxSlot}}
  ImageTriangleR_1 {
   widgetClassPath=&Image
   parent=HorizontalBox
   widget=ImageTriangleR_1
   properties {
    brush {
     bIsDynamicallyLoaded=false
     drawAs=Image
     imageType=NoImage
     mirroring=NoMirror
     resourceName=None
     resourceObject=&M_FX_TriangleScroll-Burst
     tiling=NoTile
     imageSize=<200 40>
     margin=<ltrb 0 0 0 0>
     outlineSettings {
      bUseBrushTransparency=false
      roundingType=HalfHeightRadius
      width=0
      color {colorUseRule=UseColor_Specified, specifiedColor=<rgba 0 0 0 0>}
      cornerRadii=<0 0 0 0>
     }
     tintColor {colorUseRule=UseColor_Specified, specifiedColor=<rgba 0 1 0.1907 1>}
     uVRegion {bIsValid=false, max=<0 0>, min=<0 0>}
    }
    renderTransformPivot=<0.5 1>
   }
   slot {class=HorizontalBoxSlot, properties {verticalAlignment=VAlign_Center, padding=<ltrb 0 0 -200 0>}}
  }
  HorizontalBox_0 {
   widgetClassPath=&HorizontalBox
   parent=Overlaytriangles
   widget=HorizontalBox_0
   properties {renderTransform {angle=0, scale=<1 1>, shear=<-12 0>, translation=<0 0>}}
   slot {class=OverlaySlot, properties {verticalAlignment=VAlign_Center}}
  }
  ImageTriangleL {
   widgetClassPath=&Image
   parent=HorizontalBox_0
   widget=ImageTriangleL
   properties {
    brush {
     bIsDynamicallyLoaded=false
     drawAs=Image
     imageType=NoImage
     mirroring=NoMirror
     resourceName=None
     resourceObject=/Game/UI/HUD/Materials/M_FX_TriangleSpiral
     tiling=NoTile
     imageSize=<200 64>
     margin=<ltrb 0 0 0 0>
     outlineSettings {
      bUseBrushTransparency=false
      roundingType=HalfHeightRadius
      width=0
      color {colorUseRule=UseColor_Specified, specifiedColor=<rgba 0 0 0 0>}
      cornerRadii=<0 0 0 0>
     }
     tintColor {colorUseRule=UseColor_Specified, specifiedColor=<rgba 0 1 0.1907 1>}
     uVRegion {bIsValid=false, max=<0 0>, min=<0 0>}
    }
    renderTransform {angle=0, scale=<-1 1>, shear=<0 0>, translation=<0 0>}
    renderTransformPivot=<0.5 1>
   }
   slot {class=HorizontalBoxSlot, properties {horizontalAlignment=HAlign_Right, verticalAlignment=VAlign_Center, padding=<ltrb -200 0 0 0>}}
  }
  Spacer_70 {widgetClassPath=&Spacer, parent=HorizontalBox_0, widget=Spacer_70, slot {class=HorizontalBoxSlot}}
  ImageTriangleR {
   widgetClassPath=&Image
   parent=HorizontalBox_0
   widget=ImageTriangleR
   properties {
    brush {
     bIsDynamicallyLoaded=false
     drawAs=Image
     imageType=NoImage
     mirroring=NoMirror
     resourceName=None
     resourceObject=&M_FX_TriangleSpiral
     tiling=NoTile
     imageSize=<200 64>
     margin=<ltrb 0 0 0 0>
     outlineSettings {
      bUseBrushTransparency=false
      roundingType=HalfHeightRadius
      width=0
      color {colorUseRule=UseColor_Specified, specifiedColor=<rgba 0 0 0 0>}
      cornerRadii=<0 0 0 0>
     }
     tintColor {colorUseRule=UseColor_Specified, specifiedColor=<rgba 0 1 0.1907 1>}
     uVRegion {bIsValid=false, max=<0 0>, min=<0 0>}
    }
    renderTransformPivot=<0.5 1>
   }
   slot {class=HorizontalBoxSlot, properties {verticalAlignment=VAlign_Center, padding=<ltrb 0 0 -200 0>}}
  }
  Overlay_CheckmarkContainer {
   widgetClassPath=&Overlay
   parent=OverlayCheck
   widget=Overlay_CheckmarkContainer
   properties {renderOpacity=0, visibility=HitTestInvisible}
   slot {class=OverlaySlot, properties {horizontalAlignment=HAlign_Center, verticalAlignment=VAlign_Center}}
  }
  Checkmark_Shadow {
   widgetClassPath=&Image
   bIsVariable=true
   parent=Overlay_CheckmarkContainer
   widget=Checkmark_Shadow
   properties {
    brush {
     bIsDynamicallyLoaded=false
     drawAs=Image
     imageType=NoImage
     mirroring=NoMirror
     resourceName=None
     resourceObject=/Game/Athena/HUD/Atlas/Bacchus/Confirm_Normal_L_Sprite
     tiling=NoTile
     imageSize=<50 50>
     margin=<ltrb 0 0 0 0>
     outlineSettings {
      bUseBrushTransparency=false
      roundingType=HalfHeightRadius
      width=0
      color {colorUseRule=UseColor_Specified, specifiedColor=<rgba 0 0 0 0>}
      cornerRadii=<0 0 0 0>
     }
     tintColor {colorUseRule=UseColor_Specified, specifiedColor=<rgba 0 0 0 0.8>}
     uVRegion {bIsValid=false, max=<0 0>, min=<0 0>}
    }
    colorAndOpacity=<rgba 1 1 1 0.8>
    renderTransform {angle=0, scale=<1 1>, shear=<0 0>, translation=<3 4>}
   }
   slot {class=OverlaySlot}
  }
  Checkmark {
   widgetClassPath=&Image
   bIsVariable=true
   parent=Overlay_CheckmarkContainer
   widget=Checkmark
   properties {brush {
     bIsDynamicallyLoaded=false
     drawAs=Image
     imageType=NoImage
     mirroring=NoMirror
     resourceName=None
     resourceObject=&Confirm_Normal_L_Sprite
     tiling=NoTile
     imageSize=<50 50>
     margin=<ltrb 0 0 0 0>
     outlineSettings {
      bUseBrushTransparency=false
      roundingType=HalfHeightRadius
      width=0
      color {colorUseRule=UseColor_Specified, specifiedColor=<rgba 0 0 0 0>}
      cornerRadii=<0 0 0 0>
     }
     tintColor {colorUseRule=UseColor_Specified, specifiedColor=<rgba 0.0343 1 0.227 1>}
     uVRegion {bIsValid=false, max=<0 0>, min=<0 0>}
    }}
   slot {class=OverlaySlot, properties {horizontalAlignment=HAlign_Center, verticalAlignment=VAlign_Center}}
  }
  Overlay_Flare {
   widgetClassPath=&Overlay
   parent=OverlayCheck
   widget=Overlay_Flare
   properties {visibility=HitTestInvisible, renderTransform {angle=0, scale=<0.8 0.8>, shear=<0 0>, translation=<0 0>}}
   slot {class=OverlaySlot, properties {horizontalAlignment=HAlign_Center, verticalAlignment=VAlign_Center}}
  }
  ImageSpinningSpikeball1 {
   widgetClassPath=&Image
   parent=Overlay_Flare
   widget=ImageSpinningSpikeball1
   properties {
    renderOpacity=0
    brush {
     bIsDynamicallyLoaded=false
     drawAs=Image
     imageType=NoImage
     mirroring=NoMirror
     resourceName=None
     resourceObject=/Game/Athena/HUD/Materials/Victory/M_Flare_Spikeball
     tiling=NoTile
     imageSize=<256 256>
     margin=<ltrb 0 0 0 0>
     outlineSettings {
      bUseBrushTransparency=false
      roundingType=HalfHeightRadius
      width=0
      color {colorUseRule=UseColor_Specified, specifiedColor=<rgba 0 0 0 0>}
      cornerRadii=<0 0 0 0>
     }
     tintColor {colorUseRule=UseColor_Specified, specifiedColor=<rgba 1 1 1 1>}
     uVRegion {bIsValid=false, max=<0 0>, min=<0 0>}
    }
    renderTransform {angle=13.5163, scale=<2 2>, shear=<0 0>, translation=<0 0>}
   }
   slot {class=OverlaySlot, properties {horizontalAlignment=HAlign_Center, verticalAlignment=VAlign_Center, padding=<ltrb -128 -128 -128 -128>}}
  }
  Image1Cross {
   widgetClassPath=&Image
   parent=Overlay_Flare
   widget=Image1Cross
   properties {
    renderOpacity=0
    brush {
     bIsDynamicallyLoaded=false
     drawAs=Image
     imageType=NoImage
     mirroring=NoMirror
     resourceName=None
     resourceObject=/Game/Athena/HUD/Materials/Victory/M_Cross_Flare
     tiling=NoTile
     imageSize=<128 128>
     margin=<ltrb 0 0 0 0>
     outlineSettings {
      bUseBrushTransparency=false
      roundingType=HalfHeightRadius
      width=0
      color {colorUseRule=UseColor_Specified, specifiedColor=<rgba 0 0 0 0>}
      cornerRadii=<0 0 0 0>
     }
     tintColor {colorUseRule=UseColor_Specified, specifiedColor=<rgba 1 1 1 1>}
     uVRegion {bIsValid=false, max=<0 0>, min=<0 0>}
    }
    renderTransform {angle=0, scale=<2 2>, shear=<0 0>, translation=<0 0>}
   }
   slot {class=OverlaySlot, properties {horizontalAlignment=HAlign_Center, verticalAlignment=VAlign_Center, padding=<ltrb -64 -64 -64 -64>}}
  }
  BallOverlay {
   widgetClassPath=&Overlay
   parent=OverlayTextContent
   widget=BallOverlay
   properties {renderOpacity=0, renderTransform {angle=0, scale=<1 1>, shear=<0 0>, translation=<407 0>}}
   slot {class=OverlaySlot}
  }
  BallContextualBall {
   widgetClassPath=&ContextTutorialBall
   bIsVariable=true
   parent=BallOverlay
   widget=BallContextualBall
   properties {
    hiddenVisibility=Hidden
    visibility=Hidden
    elementTag {tagName=Athena.HideForCapture}
    renderTransform {angle=0, scale=<1 1>, shear=<0 0>, translation=<-10 0>}
   }
   slot {class=OverlaySlot}
  }
  MobileCLoseButtonArea {
   widgetClassPath=&CommonBorder
   bIsVariable=true
   parent=CanvasPanel_Container
   widget=MobileCLoseButtonArea
   properties {visibility=Hidden, padding=<ltrb 0 0 0 0>, renderTransform {angle=0, scale=<1.4 1.4>, shear=<0 0>, translation=<354 -82>}}
   slot {
    class=CanvasPanelSlot
    properties {layoutData {alignment=<0.9106 0>, anchors {maximum=<1 0.5>, minimum=<1 0.5>}, offsets=<ltrb -315.149 -127.7823 205.898 209.8766>}}
   }
  }
 ]
}
