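# HUD widget that shows the current limited-time mode's name and up to four bulleted description lines.
# On Construct it caches the text blocks and their containers in arrays, then looks up the game state's playlist name in DataTable_ModeInfo, printing a message if no row exists.
# SetModeInfo fills the slots by index, reveals each initially collapsed container it uses, and breaks once descriptions exceed the slots.
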WidgetBlueprint /Game/Athena/HUD/LTM/UI_LTM_ModeInfo
graphs [EventGraph {nodes [
   n1 = SetModeInfo!() -> (OutputDelegate, then, Name:Text, Descriptions:Array of Texts) {is=title K2Node_CustomEvent}
   > "SetText (Text)"(self=Get Txt_Name() -> (Txt_Name:Common Text Object Reference), InText=@n1.Name)
   > n2 = For Each Loop with Break!(Exec:Exec, Array:Array of Texts=@n1.Descriptions, Break:Exec=@n4.else) -> (LoopBody:Exec, Array Element:Text, Array Index:int, Completed:Exec)
   n3 = Get Descriptions() -> (Descriptions:Array of Common Text Object References)
   > n4 = Branch(Condition=Is Valid Index!(TargetArray:Array of Common Text Object References=@n3, IndexToTest:int=@n2.Array Index) -> (ReturnValue:bool))
   > "SetText (Text)"(
    self="Get (a copy)"(Array:Array of Common Text Object References=@n3, Dimension 1=@n2.Array Index) -> (Output:Common Text Object Reference)
    InText=@n2
   )
   n5 = Get DescriptionContainers() -> (DescriptionContainers:Array of Vertical Box Object References)
   > Branch(Condition=Is Valid Index(TargetArray:Array of Vertical Box Object References=@n5, IndexToTest=@n2.Array Index))
   > SetVisibility(
    self="Get (a copy)"(Array:Array of Vertical Box Object References=@n5, Dimension 1=@n2.Array Index) -> (Output:Vertical Box Object Reference)
    InVisibility=HitTestInvisible
   )
   Event Construct()
   > n6 = Sequence()
   n7 = Cast To FortGameStateAthena!(execute=@n6.then_1, Object=GetGameState!() -> (ReturnValue:Game State Base Object Reference)) -> (then, CastFailed, AsFort Game State Athena:Fort Game State Athena Object Reference) {is=title K2Node_DynamicCast}
   n8 = GetCurrentPlaylistName!(self:Fort Game State Athena Object Reference=@n7) -> (ReturnValue:Name=None)
   > n9 = Get Data Table Row!(execute:Exec, DataTable:Data Table Object Reference=Get ModeInfoTable() -> (ModeInfoTable:Data Table Object Reference), RowName:Name=@n8) -> (then:Exec, RowNotFound:Exec, ReturnValue:Struct Mode Info Structure)
   n10 = Break Struct Mode Info!(Struct_ModeInfo:"Struct Mode Info Structure (by ref)"=@n9) -> (LTMName_9_#1:Text, Descriptions_13_#2:Array of Texts) {is=title K2Node_BreakStruct}
   > SetModeInfo!(execute, self, Name:Text=@n10, Descriptions:Array of Texts=@n10.Descriptions_13_#2) -> (then) {is=title K2Node_CallFunction}
   n11 = "To Text (Name)"!(InName:Name=@n8) -> (ReturnValue:Text)
   PrintString!(
    execute=@n9.RowNotFound
    InString="To String (Text)"(InText=Format Text!(Format="LOCTEXT(\"{PlaylistName} Row not found\")", PlaylistName:Text=@n11) -> (Result))
    bPrintToScreen
    bPrintToLog
    TextColor
    Duration
    Key
   ) -> (then)
   Set Descriptions(
    execute=@n6.then_0
    Descriptions:Array of Common Text Object References=Make Array!(
     "[0]":Common Text Object Reference=Get Txt_Description() -> (Txt_Description:Common Text Object Reference)
     "[1]":Common Text Object Reference=Get Txt_Description_1() -> (Txt_Description_1:Common Text Object Reference)
     "[2]":Common Text Object Reference=Get Txt_Description_2() -> (Txt_Description_2:Common Text Object Reference)
     "[3]":Common Text Object Reference=Get Txt_Description_3() -> (Txt_Description_3:Common Text Object Reference)
    ) -> (Array:Array of Common Text Object References)
   ) -> (Output_Get:Array of Common Text Object References)
   > Set DescriptionContainers(DescriptionContainers:Array of Vertical Box Object References=Make Array(
     "[0]":Vertical Box Object Reference=Get VerticalBox_Description() -> (VerticalBox_Description:Vertical Box Object Reference)
     "[1]":Vertical Box Object Reference=Get VerticalBox_Description_1() -> (VerticalBox_Description_1:Vertical Box Object Reference)
     "[2]":Vertical Box Object Reference=Get VerticalBox_Description_2() -> (VerticalBox_Description_2:Vertical Box Object Reference)
     "[3]":Vertical Box Object Reference=Get VerticalBox_Description_3() -> (VerticalBox_Description_3:Vertical Box Object Reference)
    ) -> (Array:Array of Vertical Box Object References)) -> (Output_Get:Array of Vertical Box Object References)
  ]}]
variables [
 DescriptionContainers {type="array<VerticalBox>", default []}
 Descriptions {type="array<CommonTextBlock>", default []}
 ModeInfoTable {type=DataTable, default=&/DataTable_ModeInfo}
]
widgets {
 info {inheritedWidgetCount=0, parentClass=&UserWidget, rootWidgetClass=&Overlay, widgetCount=21}
 widgets [
  Overlay_21 {widgetClassPath=&Overlay, widget=Overlay_21}
  VerticalBox_103 {
   widgetClassPath=&VerticalBox
   parent=Overlay_21
   widget=VerticalBox_103
   slot {class=OverlaySlot, properties {horizontalAlignment=HAlign_Fill, verticalAlignment=VAlign_Fill, padding=<ltrb 0 50 0 0>}}
  }
  VerticalBox_Name {
   widgetClassPath=&VerticalBox
   bIsVariable=true
   parent=VerticalBox_103
   widget=VerticalBox_Name
   properties {renderTransform {angle=0, scale=<1 1>, shear=<-8 0>, translation=<0 0>}}
   slot {class=VerticalBoxSlot, properties {padding=<ltrb 15 0 0 0>}}
  }
  Txt_Name {
   widgetClassPath=&CommonTextBlock
   bIsVariable=true
   parent=VerticalBox_Name
   widget=Txt_Name
   properties {
    style=/Game/Athena/HUD/TextStyles/TextStyle-HUD-MapScreen_Title.TextStyle-HUD-MapScreen_Title_C
    text=LTM Name
    textTransformPolicy=ToUpper
    font {
     bForceMonospaced=false
     bMaterialIsStencil=false
     fontMaterial=null
     fontObject=&Burbank
     letterSpacing=0
     monospacedWidth=1
     size=18
     skewAmount=0
     typefaceFontName=BurbankRegular-Black
     outlineSettings {
      bApplyOutlineToDropShadows=true
      bMiteredCorners=false
      bSeparateFillAlpha=true
      outlineMaterial=&M-FontShadow
      outlineSize=3
      outlineColor=<rgba 0 0 0 0.5>
     }
    }
    shadowOffset=<0 0>
   }
   slot {class=VerticalBoxSlot}
  }
  Image_2 {
   widgetClassPath=&Image
   bIsVariable=true
   parent=VerticalBox_Name
   widget=Image_2
   properties {
    brush {
     bIsDynamicallyLoaded=false
     drawAs=Image
     imageType=NoImage
     mirroring=NoMirror
     resourceName=None
     resourceObject=null
     tiling=NoTile
     imageSize=<2 2>
     margin=<ltrb 0 0 0 0>
     outlineSettings {
      bUseBrushTransparency=false
      roundingType=HalfHeightRadius
      width=0
      color {colorUseRule=UseColor_Specified, specifiedColor=<rgba 0 0 0 0>}
      cornerRadii=<0 0 0 0>
     }
     tintColor {colorUseRule=UseColor_Specified, specifiedColor=<rgba 1 1 1 1>}
     uVRegion {bIsValid=false, max=<0 0>, min=<0 0>}
    }
    colorAndOpacity=<rgba 1 1 1 0.1>
   }
   slot {class=VerticalBoxSlot, properties {padding=<ltrb 0 10 0 10>}}
  }
  VerticalBox_Description {
   widgetClassPath=&VerticalBox
   bIsVariable=true
   parent=VerticalBox_103
   widget=VerticalBox_Description
   properties {visibility=Collapsed}
   slot {class=VerticalBoxSlot, properties {padding=<ltrb 0 20 0 0>}}
  }
  HorizontalBox_1 {widgetClassPath=&HorizontalBox, parent=VerticalBox_Description, widget=HorizontalBox_1, slot {class=VerticalBoxSlot}}
  BulletPoint_01 {
   widgetClassPath=&Image
   bIsVariable=true
   parent=HorizontalBox_1
   widget=BulletPoint_01
   properties {
    visibility=HitTestInvisible
    brush {
     bIsDynamicallyLoaded=false
     drawAs=Image
     imageType=NoImage
     mirroring=NoMirror
     resourceName=None
     resourceObject=/Game/UI/Foundation/Textures/LoadingScreens/Icons/FORT_Load_Missions_PH4-256
     tiling=NoTile
     imageSize=<64 64>
     margin=<ltrb 0 0 0 0>
     outlineSettings {
      bUseBrushTransparency=false
      roundingType=HalfHeightRadius
      width=0
      color {colorUseRule=UseColor_Specified, specifiedColor=<rgba 0 0 0 0>}
      cornerRadii=<0 0 0 0>
     }
     tintColor {colorUseRule=UseColor_Specified, specifiedColor=<rgba 1 1 1 1>}
     uVRegion {bIsValid=false, max=<0 0>, min=<0 0>}
    }
   }
   slot {class=HorizontalBoxSlot, properties {horizontalAlignment=HAlign_Center, verticalAlignment=VAlign_Top, padding=<ltrb 0 -17 0 0>}}
  }
  Txt_Description {
   widgetClassPath=&CommonTextBlock
   bIsVariable=true
   parent=HorizontalBox_1
   widget=Txt_Description
   properties {
    autoWrapText=true
    style=/Game/Athena/HUD/TextStyles/TextStyle-HUD-MapScreen_SubTitle.TextStyle-HUD-MapScreen_SubTitle_C
    text=LTM DESCRIPTION LTM DESCRIPTION LTM DESCRIPTION LTM DESCRIPTION LTM DESCRIPTION LTM DESCRIPTION LTM DESCRIPTION LTM DESCRIPTION LTM DESCRIPTION
    textTransformPolicy=ToUpper
    visibility=HitTestInvisible
    font {
     bForceMonospaced=false
     bMaterialIsStencil=false
     fontMaterial=null
     fontObject=&BurbankSmall
     letterSpacing=0
     monospacedWidth=1
     size=15
     skewAmount=0
     typefaceFontName=Black
     outlineSettings {
      bApplyOutlineToDropShadows=false
      bMiteredCorners=false
      bSeparateFillAlpha=false
      outlineMaterial=null
      outlineSize=0
      outlineColor=<rgba 0 0 0 1>
     }
    }
    shadowOffset=<0 0>
   }
   slot {class=HorizontalBoxSlot, properties {padding=<ltrb 0 0 0 15>, size {sizeRule=Fill, value=1}}}
  }
  VerticalBox_Description_1 {
   widgetClassPath=&VerticalBox
   bIsVariable=true
   parent=VerticalBox_103
   widget=VerticalBox_Description_1
   properties {visibility=Collapsed}
   slot {class=VerticalBoxSlot}
  }
  HorizontalBox_2 {widgetClassPath=&HorizontalBox, parent=VerticalBox_Description_1, widget=HorizontalBox_2, slot {class=VerticalBoxSlot}}
  BulletPoint {
   widgetClassPath=&Image
   bIsVariable=true
   parent=HorizontalBox_2
   widget=BulletPoint
   properties {
    visibility=HitTestInvisible
    brush {
     bIsDynamicallyLoaded=false
     drawAs=Image
     imageType=NoImage
     mirroring=NoMirror
     resourceName=None
     resourceObject=&FORT_Load_Missions_PH4-256
     tiling=NoTile
     imageSize=<64 64>
     margin=<ltrb 0 0 0 0>
     outlineSettings {
      bUseBrushTransparency=false
      roundingType=HalfHeightRadius
      width=0
      color {colorUseRule=UseColor_Specified, specifiedColor=<rgba 0 0 0 0>}
      cornerRadii=<0 0 0 0>
     }
     tintColor {colorUseRule=UseColor_Specified, specifiedColor=<rgba 1 1 1 1>}
     uVRegion {bIsValid=false, max=<0 0>, min=<0 0>}
    }
   }
   slot {class=HorizontalBoxSlot, properties {horizontalAlignment=HAlign_Center, verticalAlignment=VAlign_Top, padding=<ltrb 0 -17 0 0>}}
  }
  Txt_Description_1 {
   widgetClassPath=&CommonTextBlock
   bIsVariable=true
   parent=HorizontalBox_2
   widget=Txt_Description_1
   properties {
    autoWrapText=true
    style=&TextStyle-HUD-MapScreen_SubTitle_C
    text=LTM DESCRIPTION
    textTransformPolicy=ToUpper
    visibility=HitTestInvisible
    font {
     bForceMonospaced=false
     bMaterialIsStencil=false
     fontMaterial=null
     fontObject=&BurbankSmall
     letterSpacing=0
     monospacedWidth=1
     size=15
     skewAmount=0
     typefaceFontName=Black
     outlineSettings {
      bApplyOutlineToDropShadows=false
      bMiteredCorners=false
      bSeparateFillAlpha=false
      outlineMaterial=null
      outlineSize=0
      outlineColor=<rgba 0 0 0 1>
     }
    }
    shadowOffset=<0 0>
   }
   slot {class=HorizontalBoxSlot, properties {padding=<ltrb 0 0 0 15>, size {sizeRule=Fill, value=1}}}
  }
  VerticalBox_Description_2 {
   widgetClassPath=&VerticalBox
   bIsVariable=true
   parent=VerticalBox_103
   widget=VerticalBox_Description_2
   properties {visibility=Collapsed}
   slot {class=VerticalBoxSlot}
  }
  HorizontalBox_3 {widgetClassPath=&HorizontalBox, parent=VerticalBox_Description_2, widget=HorizontalBox_3, slot {class=VerticalBoxSlot}}
  BulletPoint_1 {
   widgetClassPath=&Image
   bIsVariable=true
   parent=HorizontalBox_3
   widget=BulletPoint_1
   properties {
    visibility=HitTestInvisible
    brush {
     bIsDynamicallyLoaded=false
     drawAs=Image
     imageType=NoImage
     mirroring=NoMirror
     resourceName=None
     resourceObject=&FORT_Load_Missions_PH4-256
     tiling=NoTile
     imageSize=<64 64>
     margin=<ltrb 0 0 0 0>
     outlineSettings {
      bUseBrushTransparency=false
      roundingType=HalfHeightRadius
      width=0
      color {colorUseRule=UseColor_Specified, specifiedColor=<rgba 0 0 0 0>}
      cornerRadii=<0 0 0 0>
     }
     tintColor {colorUseRule=UseColor_Specified, specifiedColor=<rgba 1 1 1 1>}
     uVRegion {bIsValid=false, max=<0 0>, min=<0 0>}
    }
   }
   slot {class=HorizontalBoxSlot, properties {horizontalAlignment=HAlign_Center, verticalAlignment=VAlign_Top, padding=<ltrb 0 -17 0 0>}}
  }
  Txt_Description_2 {
   widgetClassPath=&CommonTextBlock
   bIsVariable=true
   parent=HorizontalBox_3
   widget=Txt_Description_2
   properties {
    autoWrapText=true
    style=&TextStyle-HUD-MapScreen_SubTitle_C
    text=LTM DESCRIPTION
    textTransformPolicy=ToUpper
    visibility=HitTestInvisible
    font {
     bForceMonospaced=false
     bMaterialIsStencil=false
     fontMaterial=null
     fontObject=&BurbankSmall
     letterSpacing=0
     monospacedWidth=1
     size=15
     skewAmount=0
     typefaceFontName=Black
     outlineSettings {
      bApplyOutlineToDropShadows=false
      bMiteredCorners=false
      bSeparateFillAlpha=false
      outlineMaterial=null
      outlineSize=0
      outlineColor=<rgba 0 0 0 1>
     }
    }
    shadowOffset=<0 0>
   }
   slot {class=HorizontalBoxSlot, properties {padding=<ltrb 0 0 0 15>, size {sizeRule=Fill, value=1}}}
  }
  VerticalBox_Description_3 {
   widgetClassPath=&VerticalBox
   bIsVariable=true
   parent=VerticalBox_103
   widget=VerticalBox_Description_3
   properties {visibility=Collapsed}
   slot {class=VerticalBoxSlot}
  }
  HorizontalBox_4 {widgetClassPath=&HorizontalBox, parent=VerticalBox_Description_3, widget=HorizontalBox_4, slot {class=VerticalBoxSlot}}
  BulletPoint_2 {
   widgetClassPath=&Image
   bIsVariable=true
   parent=HorizontalBox_4
   widget=BulletPoint_2
   properties {
    visibility=HitTestInvisible
    brush {
     bIsDynamicallyLoaded=false
     drawAs=Image
     imageType=NoImage
     mirroring=NoMirror
     resourceName=None
     resourceObject=&FORT_Load_Missions_PH4-256
     tiling=NoTile
     imageSize=<64 64>
     margin=<ltrb 0 0 0 0>
     outlineSettings {
      bUseBrushTransparency=false
      roundingType=HalfHeightRadius
      width=0
      color {colorUseRule=UseColor_Specified, specifiedColor=<rgba 0 0 0 0>}
      cornerRadii=<0 0 0 0>
     }
     tintColor {colorUseRule=UseColor_Specified, specifiedColor=<rgba 1 1 1 1>}
     uVRegion {bIsValid=false, max=<0 0>, min=<0 0>}
    }
   }
   slot {class=HorizontalBoxSlot, properties {horizontalAlignment=HAlign_Center, verticalAlignment=VAlign_Top, padding=<ltrb 0 -17 0 0>}}
  }
  Txt_Description_3 {
   widgetClassPath=&CommonTextBlock
   bIsVariable=true
   parent=HorizontalBox_4
   widget=Txt_Description_3
   properties {
    autoWrapText=true
    style=&TextStyle-HUD-MapScreen_SubTitle_C
    text=LTM DESCRIPTION
    textTransformPolicy=ToUpper
    visibility=HitTestInvisible
    font {
     bForceMonospaced=false
     bMaterialIsStencil=false
     fontMaterial=null
     fontObject=&BurbankSmall
     letterSpacing=0
     monospacedWidth=1
     size=15
     skewAmount=0
     typefaceFontName=Black
     outlineSettings {
      bApplyOutlineToDropShadows=false
      bMiteredCorners=false
      bSeparateFillAlpha=false
      outlineMaterial=null
      outlineSize=0
      outlineColor=<rgba 0 0 0 1>
     }
    }
    shadowOffset=<0 0>
   }
   slot {class=HorizontalBoxSlot, properties {size {sizeRule=Fill, value=1}}}
  }
 ]
}
